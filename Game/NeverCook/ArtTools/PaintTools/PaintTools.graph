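# PaintTools is an editor utility panel for Fortnite character paint artists to stage, replace, select and export assets.
# Add spawns selected assets (character definitions as mannequins) at turntable origin or camera-trace hit, attaches them, optionally setting LOD, mobility and grid layout.
# Other buttons mostly run Paint Python scripts for levels, materials and cameras.
# FBX export temporarily zeroes parent or actor transforms; turntable-only buttons are disabled on other maps.
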EditorUtilityWidgetBlueprint /Game/NeverCook/ArtTools/PaintTools/PaintTools
graphs [
 DeleteActors {nodes [
   DeleteActors() {is=title K2Node_FunctionEntry}
   n1 = Get ActorsAddedAll() -> (ActorsAddedAll:Array of Actor Object References)
   > SetSelectedLevelActors(self=Get EditorActorSubsystem(), ActorsToSelect=@n1)
   > n2 = For Each Loop(Array:Array of Actor Object References=@n1) -> (Array Element:Actor Object Reference)
   Clear(execute=@n2.Completed, TargetArray:Array of Object References=Get SelectionList() -> (SelectionList:Array of Object References))
   > Clear(TargetArray:Array of Actor Object References=Get ActorsAddedAll())
   Destroy Actor(execute=@n2.LoopBody, self=@n2)
  ]}
 DisableNotGbuffer {nodes [
   DisableNotGbuffer() {is=title K2Node_FunctionEntry}
   > n1 = GetCurrentLevelName!(execute, WorldContextObject:Object Reference, bRemovePrefixString:bool=true) -> (then, ReturnValue:String)
   n2 = "Equal Exactly (String)"(A=@n1, B=Get TurntableMap() -> (TurntableMap:String))
   > n3 = Branch(Condition=@n2)
   SetIsEnabled(
    execute=[@n3.else, @n3.then]
    self=[
     Get EditorButton-PilotGbufferCamera() -> (EditorButton-PilotGbufferCamera:Editor Utility Button Object Reference)
     Get EditorButton-CloseUpCamera() -> (EditorButton-CloseUpCamera:Editor Utility Button Object Reference)
     Get EditorButton-CleanupTT() -> (EditorButton-CleanupTT:Editor Utility Button Object Reference)
    ]
    bInIsEnabled=@n2
   )
  ]}
 EventGraph {nodes [
   n1 = "On Clicked (EditorButton-Add)"() {is=title K2Node_ComponentBoundEvent}
   Clear(execute=[@n1.then, @n42.then], TargetArray:Array of Object References=Get SelectionList())
   > Clear(TargetArray:Array of Actor Object References=Get Actors to Select() -> (Actors to Select:Array of Actor Object References))
   > n2 = GetSelectedAssets()
   > n3 = For Each Loop(Array:Array of Object References=@n2) -> (Array Element:Object Reference)
   Set PlacementLocation(execute=@n3.Completed, PlacementLocation:Vector="0, 0, 0") -> (Output_Get:Vector)
   > Set counter(counter:int=1) -> (Output_Get:int=0)
   > SelectGeo()
   > Execute Python Script(PythonScript="from Paint.unreal_paint_tools.toon_utils import assign_overlay_mats_to_level_actors\r\n\r\nassign_overlay_mats_to_level_actors()")
   > SelectNothing(self:Editor Actor Subsystem Object Reference=Get EditorActorSubsystem())
   n4 = Cast To AthenaCharacterItemDefinition!(execute=@n3.LoopBody, Object=@n3) -> (then, CastFailed, AsAthena Character Item Definition:Athena Character Item Definition Object Reference) {is=title K2Node_DynamicCast}
   > n5 = meshPlacement!(execute, self, array pos:int=0) -> (then, NewParam:Vector)
   > n6 = SpawnActor Fort Player Mannequin!(
    execute:Exec
    WorldContextObject:Object Reference
    Class:Actor Class Reference=/Script/FortniteGame.FortPlayerMannequin
    SpawnTransform_Location:Vector=@n5
    SpawnTransform_Rotation:Rotator="0, 0, 0"
    SpawnTransform_Scale:Vector="1,1,1"
    CollisionHandlingOverride:ESpawnActorCollisionHandlingMethod Enum=Undefined
    TransformScaleMethod:ESpawnActorScaleMethod Enum=MultiplyWithRoot
    Owner:Actor Object Reference
    Instigator:Pawn Object Reference
   ) -> (then:Exec, ReturnValue:Fort Player Mannequin Object Reference) {is=title K2Node_SpawnActorFromClass}
   > Add(TargetArray:Array of Actor Object References=Get Actors to Select(), NewItem:"Actor Object Reference (by ref)"=@n6) {is=title K2Node_CallArrayFunction}
   > SetEditorProperty(Object=@n6, PropertyName=athenaCharacter, PropertyValue:Athena Character Item Definition Object Reference=@n4)
   > Attach Actor To Actor!(
    execute
    self:Actor Object Reference=@n6
    ParentActor:Actor Object Reference=Get Attach to Actor() -> (Attach to Actor:Actor Object Reference)
    SocketName:Name=None
    LocationRule:EAttachmentRule Enum=KeepRelative
    RotationRule:EAttachmentRule Enum=KeepRelative
    ScaleRule:EAttachmentRule Enum=KeepRelative
    bWeldSimulatedBodies:bool=TRUE
   ) -> (then, ReturnValue:bool)
   Add(execute=@n4.CastFailed, TargetArray:Array of Object References=Get SelectionList(), NewItem:"Object Reference (by ref)"=@n3)
   > n7 = meshPlacement(array pos=@n3.Array Index)
   > n8 = SpawnActorFromObject!(execute, ObjectToUse:Object Reference=@n3, Location:Vector=@n7, Rotation:Rotator, bTransient:bool=false) -> (then, ReturnValue:Actor Object Reference)
   > Add(TargetArray:Array of Actor Object References=Get Actors to Select(), NewItem:"Actor Object Reference (by ref)"=@n8)
   > Add(TargetArray=Get ActorsAddedAll(), NewItem=@n8)
   > n9 = Cast To StaticMeshActor!(execute, Object=@n8) -> (then, CastFailed, AsStatic Mesh Actor:Static Mesh Actor Object Reference) {is=title K2Node_DynamicCast}
   > n10 = Branch(Condition=Get bMinLOD() -> (bMinLOD:bool))
   n11 = Get StaticMeshComponent(self=@n9)
   > SetForcedLodModel!(execute, self:Static Mesh Component Object Reference=@n11, NewForcedLodModel:int=1) -> (then)
   > n12 = Branch(Condition=Get bMobilityMovable() -> (bMobilityMovable:bool))
   n13 = SetMobility!(execute=[@n12.then, @n10.else], self:Scene Component Object Reference=@n11, NewMobility:EComponentMobility Enum=Movable) -> (then)
   Attach Actor To Actor(
    execute=[@n13.then, @n45.then, @n12.else]
    self=@n8
    ParentActor=Get Attach to Actor()
    SocketName=None
    LocationRule=KeepRelative
    RotationRule=KeepRelative
    ScaleRule=KeepRelative
    bWeldSimulatedBodies=TRUE
   )
   > n14 = Branch(Condition=Get bArrayAssets() -> (bArrayAssets:bool))
   n15 = Cast To SkeletalMeshActor!(execute=@n9.CastFailed, Object=@n8) -> (then, CastFailed, AsSkeletal Mesh Actor:Skeletal Mesh Actor Object Reference) {is=title K2Node_DynamicCast}
   n16 = Get SkeletalMeshComponent!(self:Skeletal Mesh Actor Object Reference=@n15) -> (SkeletalMeshComponent:Skeletal Mesh Component Object Reference)
   Set Relative Location!(
    execute=@n14.then
    self:Scene Component Object Reference=[Get RootComponent!(self:Actor Object Reference=@n8) -> (RootComponent:Scene Component Object Reference), @n16, @n11]
    NewLocation:Vector=Get PlacementLocation() -> (PlacementLocation:Vector)
    bSweep:bool=false
    bTeleport:bool=false
   ) -> (then, SweepHitResult:Hit Result Structure) {is=title K2Node_CallFunction}
   > n17 = Branch(Condition="Equal (Integer)"(A=Get counter() -> (counter:int=0), B=Get PerRow() -> (PerRow:int=0)))
   n18 = BreakVector(InVec=Get PlacementLocation())
   n19 = Get Spacing() -> (Spacing:int=0)
   > Set PlacementLocation(PlacementLocation=MakeVector(Y=float - float(A=@n18.Y, B:int=@n19), Z=@n18.Z))
   > Set counter(counter=1) -> (Output_Get=0)
   n20 = BreakVector(InVec=Get PlacementLocation())
   n21 = Get Spacing() -> (Spacing=0)
   Set PlacementLocation(execute=@n17.else, PlacementLocation=MakeVector(X=float + float(A=@n20, B:int=@n21), Y=@n20.Y, Z=@n20.Z))
   > Set counter(counter=int + int(A=Get counter() -> (counter=0))) -> (Output_Get=0)
   "On Property Changed (ParentSelector)"!() -> (OutputDelegate, then, PropertyName:Name=None) {is=title K2Node_ComponentBoundEvent}
   n22 = Get Attach to Actor()
   > n23 = Is Valid(InputObject=@n22)
   PrintLog!(execute=@n23.Is Not Valid, Log:Text="LOCTEXT(\"bad\")", __WorldContext:Object Reference) -> (then)
   PrintLog(execute=@n23.Is Valid, Log="LOCTEXT(\"good\")")
   > Set Attach to Actor(Attach to Actor:Actor Object Reference=@n22) -> (Output_Get:Actor Object Reference)
   "On Clicked (EditorButton-Replace)"() {is=title K2Node_ComponentBoundEvent}
   > DeleteActors() {is=title K2Node_CallFunction}
   > add content()
   "On Clicked (EditorButton-Selection)"() {is=title K2Node_ComponentBoundEvent}
   > SelectGeo()
   "On Clicked (EditorButton-ImportDemo)"() {is=title K2Node_ComponentBoundEvent}
   > Execute Python Script(PythonScript="import importlib\r\nimport  Paint.unreal_paint_tools\r\n\r\nfrom Paint.unreal_paint_tools import static_meshes_utils\r\n\r\nimportlib.reload(static_meshes_utils)\r\n\r\nstatic_meshes_utils.run_import_demo_mesh()")
   "On Clicked (EditorButton-MaterialAssign)"() {is=title K2Node_ComponentBoundEvent}
   > Execute Python Script(PythonScript="from Paint.unreal_paint_tools.load_static_meshes_materials import main\r\n\r\nmain()")
   "On Clicked (EditorButton-SkinToneAssist)"() {is=title K2Node_ComponentBoundEvent}
   > Execute Python Script(PythonScript="import importlib\r\nimport  Paint.unreal_paint_tools\r\n\r\nfrom Paint.unreal_paint_tools import skintone_wizard\r\n\r\nimportlib.reload(skintone_wizard)\r\n\r\nskintone_wizard.main()")
   "On Clicked (EditorButton-OpenGbufferTT)"() {is=title K2Node_ComponentBoundEvent}
   > Execute Python Script(PythonScript="from Paint.unreal_paint_tools.fn_levels_utils import load_gbuffer_level\r\n\r\nload_gbuffer_level()\r\n")
   "On Clicked (EditorButton-OpenBRArt)"() {is=title K2Node_ComponentBoundEvent}
   > Execute Python Script(PythonScript="from Paint.unreal_paint_tools.fn_levels_utils import load_br_art_level\r\n\r\nload_br_art_level()\r\n")
   "On Clicked (EditorButton-SpanwPaintWidgets)"() {is=title K2Node_ComponentBoundEvent}
   > Execute Python Script(PythonScript="from Paint.unreal_paint_tools.fn_levels_utils import spawn_paint_widgets\r\n\r\nspawn_paint_widgets()\r\n")
   "On Clicked (EditorButton-PilotGbufferCamera)"() {is=title K2Node_ComponentBoundEvent}
   > ExecutePythonCommand(PythonCommand="from Paint.unreal_paint_tools.fn_levels_utils import load_gbuffer_camera\r\n\r\nload_gbuffer_camera()\r\n")
   "On Clicked (EditorButton-MedLOD0)"() {is=title K2Node_ComponentBoundEvent}
   > Execute Python Script(PythonScript="from Paint.unreal_paint_tools.static_meshes_utils import set_sm_lod_to_zero\r\n\r\nset_sm_lod_to_zero()")
   n24 = "On Check State Changed (checkboxArrayAssets)"!() -> (OutputDelegate, then, bIsChecked:bool) {is=title K2Node_ComponentBoundEvent}
   > Set bArrayAssets(bArrayAssets:bool=@n24.bIsChecked) -> (Output_Get:bool)
   "On Clicked (EditorButton-EXT_MaterialCheatSheet)"() {is=title K2Node_ComponentBoundEvent}
   > Execute Python Script(PythonScript="import webbrowser\r\n\r\nwebbrowser.open('https://confluence.it.epicgames.com/display/FORT/FN+Character+Paint+-+Parent+Material+Cheat+Sheet')")
   "On Clicked (EditorButton-FormatMaterialString)"() {is=title K2Node_ComponentBoundEvent}
   > Execute Python Script(PythonScript="import Paint.unreal_paint_tools.imp_path_parser as upa\r\n\r\nupa.main()")
   "On Clicked (ExportFBXButton)"() {is=title K2Node_ComponentBoundEvent}
   > n25 = GetSelectedLevelActors(self=Get EditorActorSubsystem())
   n26 = GetAttachParentActor!(self:Actor Object Reference="Get (a copy)"(Array:Array of Actor Object References=@n25) -> (Output:Actor Object Reference)) -> (ReturnValue:Actor Object Reference)
   n27 = IsValid(Object=@n26)
   > n28 = Branch(Condition=@n27)
   > Set OldTransform(OldTransform:Transform=Get Actor Transform(self=@n26)) -> (Output_Get:Transform)
   > n29 = Set Actor Transform!(
    execute
    self:Actor Object Reference=@n26
    NewTransform:"Transform (by ref)"=Get zeroTransform() -> (zeroTransform:Transform)
    bSweep:bool=false
    bTeleport:bool=false
   ) -> (then, SweepHitResult:Hit Result Structure, ReturnValue:bool) {is=title K2Node_CallFunction}
   Execute Python Script(execute=[@n29.then, @n32.Completed], PythonScript="import unreal\r\nunreal.EditorLoadingAndSavingUtils.export_scene(True)")
   > n30 = Branch(Condition=@n27)
   > Set Actor Transform(self=@n26, NewTransform=Get OldTransform() -> (OldTransform:Transform), bSweep=false, bTeleport=false)
   n31 = For Each Loop(Exec=@n30.else, Array:Array of Actor Object References=@n25) -> (Array Element:Actor Object Reference)
   > Set Actor Transform(
    self=@n31
    NewTransform="Get (a ref)"(Array:Array of Transforms=Get nonParentTransform() -> (nonParentTransform:Array of Transforms), Dimension 1=@n31.Array Index) -> (Output:"Transform (by ref)")
    bSweep=false
    bTeleport=false
   )
   n32 = For Each Loop(Exec=@n28.else, Array=@n25)
   > Insert!(
    execute
    TargetArray:Array of Transforms=Get nonParentTransform()
    NewItem:"Transform (by ref)"=Get Actor Transform(self=@n32)
    Index:int=@n32.Array Index
   ) -> (then) {is=title K2Node_CallArrayFunction}
   > Set Actor Transform(self=@n32, NewTransform=Get zeroTransform(), bSweep=false, bTeleport=false)
   "On Clicked (EditorButton-CloseUpCamera)"() {is=title K2Node_ComponentBoundEvent}
   > Execute Python Script(PythonScript="import Paint.unreal_paint_tools.fn_levels_utils as upa\r\n\r\nimport importlib\r\n\r\nimportlib.reload(upa)\r\n\r\nupa.toggle_camera_actor_angles()\r\n\r\n")
   "On Clicked (EditorButton-Delete)"() {is=title K2Node_ComponentBoundEvent}
   > DeleteActors()
   "On Clicked (EditorButton-CleanupTT)"() {is=title K2Node_ComponentBoundEvent}
   > n33 = GetAllLevelActors(self=Get EditorActorSubsystem())
   > n34 = For Each Loop(Array=@n33)
   > n35 = Switch on String!(execute:Exec, Selection:String=GetDisplayName(Object=@n34)) -> (Default:Exec, ReplaceMe:Exec, SK_Mannequin:Exec) {pinNames [ReplaceMe,SK_Mannequin]}
   Destroy Actor(execute=[@n35.ReplaceMe, @n35.SK_Mannequin], self=@n34)
   n36 = "On Text Changed (InputSpacing)"!() -> (OutputDelegate, then, Text:"Text (by ref)") {is=title K2Node_ComponentBoundEvent}
   n37 = "To String (Text)"(InText=@n36.Text)
   > Branch(Condition=IsNumeric!(SourceString:String=@n37) -> (ReturnValue:bool))
   > Set Spacing(Spacing:int=String To Integer(InString=@n37)) -> (Output_Get:int=0)
   n38 = "On Text Changed (InuptRows)"!() -> (OutputDelegate, then, Text:"Text (by ref)") {is=title K2Node_ComponentBoundEvent}
   n39 = "To String (Text)"(InText=@n38.Text)
   > Branch(Condition=IsNumeric(SourceString=@n39))
   > Set PerRow(PerRow:int=String To Integer(InString=@n39)) -> (Output_Get:int=0)
   n40 = "On Check State Changed (checkboxMinLOD)"!() -> (OutputDelegate, then, bIsChecked:bool) {is=title K2Node_ComponentBoundEvent}
   > Set bMinLOD(bMinLOD:bool=@n40.bIsChecked) -> (Output_Get:bool)
   n41 = "On Check State Changed (checkboxMinLOD_1)"!() -> (OutputDelegate, then, bIsChecked:bool) {is=title K2Node_ComponentBoundEvent}
   > Set bMobilityMovable(bMobilityMovable:bool=@n41.bIsChecked) -> (Output_Get:bool)
   n42 = add content() {is=title K2Node_CustomEvent}
   Event PreConstruct()
   > DisableNotGbuffer() {is=title K2Node_CallFunction}
   Event Construct()
   > SetObject(self=Get ParentSelector() -> (ParentSelector:Details View Object Reference), NewObject=Self-Reference())
   > SetObject(
    self=[
     Get OutputFolder_DV() -> (OutputFolder_DV:Details View Object Reference)
     Get MRQSettingLocator() -> (MRQSettingLocator:Details View Object Reference)
     Get MeshOverride_DV() -> (MeshOverride_DV:Details View Object Reference)
    ]
    NewObject=Self-Reference()
   )
   > n43 = GetAllActorsOfClass!(
    execute
    WorldContextObject:Object Reference
    ActorClass:Actor Class Reference=/Gbuffer_Turntable/FN_PublicTurntable_Assets/BP/BP_Turntable_GBuffers.BP_Turntable_GBuffers_C
   ) -> (then, OutActors:Array of BP Turntable GBuffers Object References)
   > Branch(Condition=Is Not Empty!(TargetArray:Array of BP Turntable GBuffers Object References=@n43) -> (ReturnValue:bool))
   > n44 = GetAllActorsWithTag!(execute, WorldContextObject:Object Reference, Tag:Name=TurnTable_UserObject) -> (then, OutActors:Array of Actor Object References)
   > Branch(Condition=Is Not Empty(TargetArray:Array of Actor Object References=@n44))
   > Set Attach to Actor(Attach to Actor="Get (a copy)"(Array=@n44))
   Event Tick()
   Get UVEditorSubsystem() -> (ReturnValue:UVEditor Subsystem Object Reference) {is=title K2Node_GetEditorSubsystem}
   SetForcedLOD!(execute=@n46.then, self:Skinned Mesh Component Object Reference=@n16, InNewForcedLOD:int=1) -> (then)
   > Branch(Condition=Get bMobilityMovable())
   > n45 = SetMobility(self=@n16, NewMobility=Movable)
   n46 = Branch(execute=@n15.then, Condition=Get bMinLOD())
  ]}
 NewEventDispatcher {nodes [NewEventDispatcher() {is=title K2Node_FunctionEntry}]}
 SelectGeo {nodes [
   SelectGeo() {is=title K2Node_FunctionEntry}
   n1 = Get Actors to Select()
   > n2 = For Each Loop(Array=@n1)
   SetSelectedLevelActors(execute=@n2.Completed, self=Get EditorActorSubsystem(), ActorsToSelect=@n1)
   PrintString(execute=@n2.LoopBody, InString=GetDisplayName(Object=@n2))
  ]}
 SetIconTexture {nodes [
   SetIconTexture() {is=title K2Node_FunctionEntry}
   > n1 = CreateDynamicMaterialInstance(Parent=&/MI_PaintTools_button, OptionalName=None, CreationFlags=None)
   > n2 = Set ButtonMaterial(ButtonMaterial:Material Instance Dynamic Object Reference=@n1) -> (Output_Get:Material Instance Dynamic Object Reference)
   n3 = Get T Select Icon() -> (T Select Icon:Texture Object Reference)
   > SetTextureParameterValue(self=[Get ButtonMaterial() -> (ButtonMaterial:Material Instance Dynamic Object Reference), @n2], ParameterName=Center Icon, Value=@n3)
   n4 = Get EditorButton-Selection() -> (EditorButton-Selection:Editor Utility Button Object Reference)
   > Set members in Slate Brush!(
    execute:Exec
    StructRef:"Slate Brush Structure (by ref)"=Break Button Style!(ButtonStyle:"Button Style Structure (by ref)"=Get WidgetStyle!(self:Button Object Reference=@n4) -> (WidgetStyle:Button Style Structure)) -> (
     Normal:Slate Brush Structure
     Hovered:Slate Brush Structure
     Pressed:Slate Brush Structure
     Disabled:Slate Brush Structure
     NormalForeground:Slate Color Structure
     HoveredForeground:Slate Color Structure
     PressedForeground:Slate Color Structure
     DisabledForeground:Slate Color Structure
     NormalPadding:Margin Structure
     PressedPadding:Margin Structure
     PressedSlateSound:Slate Sound Structure
     ClickedSlateSound:Slate Sound Structure
     HoveredSlateSound:Slate Sound Structure
    ) {is=title K2Node_BreakStruct}
    ResourceObject:Object Reference=@n3
   ) -> (then:Exec, StructOut:"Slate Brush Structure (by ref)") {is=title K2Node_SetFieldsInStruct}
  ]}
 StateVisuals {nodes [
   n1 = StateVisuals!() -> (then, Condition:bool=~) {is=title K2Node_FunctionEntry}
   > n2 = Branch(Condition=@n1)
   n3 = Get TextSpacing() -> (TextSpacing:Text Object Reference)
   > Set ColorAndOpacity!(
    execute
    ColorAndOpacity:Slate Color Structure=Get enabledColorText() -> (enabledColorText:Slate Color Structure)
    self:Text Object Reference=@n3
   ) -> (then, Output_Get:Slate Color Structure)
   Set ColorAndOpacity(execute=@n2.else, ColorAndOpacity=Get disabledColorText() -> (disabledColorText:Slate Color Structure), self=@n3)
   Get checkboxArrayAssets() -> (checkboxArrayAssets:Editor Utility Check Box Object Reference)
  ]}
 getSelected {nodes [
   getSelected() {is=title K2Node_FunctionEntry}
   > n1 = GetSelectedAssets()
   > n2 = For Each Loop(Array:Array of Object References=@n1) -> (Array Element:Object Reference)
   > n3 = SpawnActorFromObject(ObjectToUse=@n2, Location="0, 0, 0", bTransient=false)
   > Attach Actor To Actor(
    self=@n3
    ParentActor=Get Attach to Actor()
    SocketName=None
    LocationRule=KeepRelative
    RotationRule=KeepRelative
    ScaleRule=KeepRelative
    bWeldSimulatedBodies=TRUE
   )
   > n4 = Cast To StaticMeshActor(Object=@n3)
   n5 = Get StaticMeshComponent(self=@n4)
   > SetForcedLodModel(self=@n5, NewForcedLodModel=1)
   > SetMobility(self=@n5, NewMobility=Movable)
   n6 = Cast To SkeletalMeshActor(execute=@n4.CastFailed, Object=@n3)
   > n7 = SpawnActorFromObject(ObjectToUse=@n6, Location="0, 0, 0", bTransient=false)
   > n8 = Cast To SkeletalMeshActor(Object=@n7)
   n9 = Get SkeletalMeshComponent(self=@n8)
   > SetForcedLOD(self=@n9, InNewForcedLOD=1)
   > SetMobility(self=@n9, NewMobility=Movable)
  ]}
 meshPlacement {nodes [
   n1 = meshPlacement!() -> (then, array pos:int) {is=title K2Node_FunctionEntry}
   > n2 = GetCurrentLevelName(bRemovePrefixString=true)
   > n3 = Branch(Condition="Equal Exactly (String)"(A=@n2, B=Get TurntableMap()))
   > n4 = Set MeshPosition(MeshPosition:Vector="0, 0, 0") -> (Output_Get:Vector)
   Return Node!(execute=[@n7.else, @n8.then, @n4.then], NewParam:Vector=Get MeshPosition() -> (MeshPosition:Vector))
   n5 = Get Editor Camera Info!(execute=@n3.else) -> (then, CameraLocation:Vector, ForwardDirection:Vector, FOV:float, ReturnValue:bool) {is=title K2Node_CallFunction}
   > n6 = Line Trace By Channel!(
    execute
    WorldContextObject:Object Reference
    Start:Vector=@n5
    End:Vector=vector + vector(A=@n5, B=vector * vector(A=@n5.ForwardDirection, B:Vector="9999,9999,9999") -> (ReturnValue=~))
    TraceChannel:ETraceTypeQuery Enum
    bTraceComplex:bool=false
    ActorsToIgnore:Array of Actor Object References
    DrawDebugType:EDrawDebugTrace Enum
    bIgnoreSelf:bool
    TraceColor:Linear Color Structure
    TraceHitColor:Linear Color Structure
    DrawTime:float
   ) -> (
    then
    OutHit_bBlockingHit:bool
    OutHit_bInitialOverlap:bool
    OutHit_Time:float
    OutHit_Distance:float
    OutHit_Location:Vector
    OutHit_ImpactPoint:Vector
    OutHit_Normal:Vector
    OutHit_ImpactNormal:Vector
    OutHit_PhysMat:Physical Material Object Reference
    OutHit_HitActor:Actor Object Reference
    OutHit_HitComponent:Primitive Component Object Reference
    OutHit_HitBoneName:Name=None
    OutHit_BoneName:Name=None
    OutHit_HitItem:int=0
    OutHit_ElementIndex:int=0
    OutHit_FaceIndex:int=0
    OutHit_TraceStart:Vector
    OutHit_TraceEnd:Vector
    ReturnValue:bool
   )
   > n7 = Branch(Condition="Equal (Integer)"(A=@n1))
   > n8 = Set MeshPosition(MeshPosition=@n6.OutHit_Location)
  ]}
]
variables [
 Actors to Select {type="array<Actor>"}
 ActorsAddedAll {type="array<Actor>"}
 Attach to Actor {type=Actor, default=null}
 bArrayAssets {type=boolean}
 bMinLOD {type=boolean, default=true}
 bMobilityMovable {type=boolean, default=true}
 ButtonMaterial {type=MaterialInstanceDynamic}
 character {type=AthenaCharacterItemDefinition, default=null}
 cleanUp {type="array<Actor>"}
 Cols {type=integer, default=8}
 counter {type=integer, default=1}
 disabledColorText {type=SlateColor, default {colorUseRule=UseColor_Specified, specifiedColor=<rgba 0.85 0.85 0.85 0.2>}}
 enabledColorText {type=SlateColor, default {colorUseRule=UseColor_Specified, specifiedColor=<rgba 0.85 0.85 0.85 1>}}
 isToon {type=boolean}
 MeshOverride_DV {type=DetailsView, default=null}
 MeshPosition {type=Vector}
 MRQSettingLocator {type=DetailsView, default=null}
 NewEventDispatcher {default=()}
 NewVar {type=boolean}
 nonParentTransform {type="array<Transform>", default []}
 OldTransform {type=Transform, default {location=<0 0 0>, rotation=<rot 0 0 0>, scale=<1 1 1>}}
 OutputFolder_DV {type=DetailsView, default=null}
 PerRow {type=integer, default=8}
 PlacementLocation {type=Vector}
 Rotation Slider Is Enabled {type=boolean, default=true}
 SelectedActors {type="array<Actor>"}
 SelectedActorsString {type="array<string>", default []}
 SelectionList {type="array<Object>", default []}
 Spacing {type=integer, default=512}
 T Select Icon {type=Texture, default=&/T_PaintTools_Select}
 Tint {type=SlateColor, default {colorUseRule=UseColor_Specified, specifiedColor=<rgba 1 0.26 0.8208 1>}}
 TurntableMap {type=string, default=Turntable_gbuffers}
 VersionInfo {type=GbufferTT_VersionInfo, default {current Version=2.4}}
 zeroTransform {type=Transform, default {location=<0 0 0>, rotation=<rot 0 0 0>, scale=<1 1 1>}}
]
widgets {
 info {rootWidgetClass=&StackBox, widgetCount=91}
 widgets [
  StackBox_0 {widgetClassPath=&StackBox, widget=StackBox_0}
  SizeBox_57 {
   widgetClassPath=&SizeBox
   parent=StackBox_0
   widget=SizeBox_57
   properties {bOverride_WidthOverride=true, widthOverride=480}
   slot {class=StackBoxSlot, properties {horizontalAlignment=HAlign_Center, verticalAlignment=VAlign_Top, size {sizeRule=Fill, value=1}}}
  }
  PANELVWRAP {
   widgetClassPath=&VerticalBox
   parent=SizeBox_57
   widget=PANELVWRAP
   slot {class=SizeBoxSlot, properties {verticalAlignment=VAlign_Center}}
  }
  MAINBUTTON {
   widgetClassPath=&VerticalBox
   parent=PANELVWRAP
   widget=MAINBUTTON
   slot {class=VerticalBoxSlot, properties {verticalAlignment=VAlign_Center}}
  }
  SizeBox_149 {widgetClassPath=&SizeBox, parent=MAINBUTTON, widget=SizeBox_149, slot {class=VerticalBoxSlot}}
  HorizontalBox_103 {
   widgetClassPath=&HorizontalBox
   parent=SizeBox_149
   widget=HorizontalBox_103
   slot {class=SizeBoxSlot, properties {verticalAlignment=VAlign_Center}}
  }
  EditorButton-Add {
   widgetClassPath=&EditorUtilityButton
   parent=HorizontalBox_103
   widget=EditorButton-Add
   properties {widgetStyle {
     clickedSlateSound {resourceObject=null}
     disabled {
      bIsDynamicallyLoaded=false
      drawAs=Box
      imageType=NoImage
      mirroring=NoMirror
      resourceName=None
      resourceObject=/Game/NeverCook/ArtTools/AddMeshToScene/MI_AddMeshToScene_button
      tiling=NoTile
      imageSize=<0 0>
      margin=<ltrb 0.48 0.48 0.48 0.48>
      outlineSettings {
       bUseBrushTransparency=false
       roundingType=FixedRadius
       width=1
       color {colorUseRule=UseColor_ColorTable, specifiedColor=<rgba 0 0 0 0>}
       cornerRadii=<4 4 4 4>
      }
      tintColor {colorUseRule=UseColor_Specified, specifiedColor=<rgba 0.0395 0.0395 0.0395 1>}
      uVRegion {bIsValid=false, max=<0 0>, min=<0 0>}
     }
     disabledForeground {colorUseRule=UseColor_ColorTable, specifiedColor=<rgba 0 0 0 0>}
     hovered {
      bIsDynamicallyLoaded=false
      drawAs=Box
      imageType=NoImage
      mirroring=NoMirror
      resourceName=None
      resourceObject=&MI_AddMeshToScene_button
      tiling=NoTile
      imageSize=<32 32>
      margin=<ltrb 0.48 0.48 0.48 0.48>
      outlineSettings {
       bUseBrushTransparency=false
       roundingType=FixedRadius
       width=1
       color {colorUseRule=UseColor_ColorTable, specifiedColor=<rgba 0 0 0 0>}
       cornerRadii=<4 4 4 4>
      }
      tintColor {colorUseRule=UseColor_Specified, specifiedColor=<rgba 1 0.6198 0.1014 1>}
      uVRegion {bIsValid=false, max=<0 0>, min=<0 0>}
     }
     hoveredForeground {colorUseRule=UseColor_ColorTable, specifiedColor=<rgba 0 0 0 0>}
     hoveredSlateSound {resourceObject=null}
     normal {
      bIsDynamicallyLoaded=false
      drawAs=Box
      imageType=FullColor
      mirroring=NoMirror
      resourceName=None
      resourceObject=&MI_PaintTools_button
      tiling=NoTile
      imageSize=<32 32>
      margin=<ltrb 0.48 0.48 0.48 0.48>
      outlineSettings {
       bUseBrushTransparency=false
       roundingType=FixedRadius
       width=1
       color {colorUseRule=UseColor_ColorTable, specifiedColor=<rgba 0 0 0 0>}
       cornerRadii=<4 4 4 4>
      }
      tintColor {colorUseRule=UseColor_Specified, specifiedColor=<rgba 0.15 0.15 0.15 1>}
      uVRegion {bIsValid=false, max=<0 0>, min=<0 0>}
     }
     normalForeground {colorUseRule=UseColor_ColorTable, specifiedColor=<rgba 0 0 0 0>}
     normalPadding=<ltrb 12 1.5 12 1.5>
     pressed {
      bIsDynamicallyLoaded=false
      drawAs=Box
      imageType=NoImage
      mirroring=NoMirror
      resourceName=None
      resourceObject=&MI_AddMeshToScene_button
      tiling=NoTile
      imageSize=<32 32>
      margin=<ltrb 0.48 0.48 0.48 0.48>
      outlineSettings {
       bUseBrushTransparency=false
       roundingType=FixedRadius
       width=1
       color {colorUseRule=UseColor_ColorTable, specifiedColor=<rgba 0 0 0 0>}
       cornerRadii=<4 4 4 4>
      }
      tintColor {colorUseRule=UseColor_Foreground, specifiedColor=<rgba 0.0395 0.0395 0.0395 1>}
      uVRegion {bIsValid=false, max=<0 0>, min=<0 0>}
     }
     pressedForeground {colorUseRule=UseColor_ColorTable, specifiedColor=<rgba 0 0 0 0>}
     pressedPadding=<ltrb 12 2.5 12 0.5>
     pressedSlateSound {resourceObject=null}
    }}
   slot {class=HorizontalBoxSlot, properties {verticalAlignment=VAlign_Center, padding=<ltrb 0 4 0 4>, size {sizeRule=Fill, value=1}}}
  }
  TextBlock_77 {
   widgetClassPath=&TextBlock
   parent=EditorButton-Add
   widget=TextBlock_77
   properties {text=ADD TO SCENE}
   slot {class=ButtonSlot, properties {padding=<ltrb 4 20 4 24>}}
  }
  VerticalBox_214 {widgetClassPath=&VerticalBox, parent=HorizontalBox_103, widget=VerticalBox_214, slot {class=HorizontalBoxSlot}}
  HorizontalBox_174 {
   widgetClassPath=&HorizontalBox
   parent=VerticalBox_214
   widget=HorizontalBox_174
   slot {
    class=VerticalBoxSlot
    properties {horizontalAlignment=HAlign_Left, verticalAlignment=VAlign_Center, padding=<ltrb 8 0 4 0>, size {sizeRule=Fill, value=1}}
   }
  }
  SizeBox_1 {
   widgetClassPath=&SizeBox
   parent=HorizontalBox_174
   widget=SizeBox_1
   properties {bOverride_HeightOverride=true, bOverride_WidthOverride=true, heightOverride=42, widthOverride=42}
   slot {
    class=HorizontalBoxSlot
    properties {horizontalAlignment=HAlign_Center, verticalAlignment=VAlign_Center, padding=<ltrb 4 0 0 0>, size {sizeRule=Fill, value=1}}
   }
  }
  EditorButton-Selection {
   widgetClassPath=&EditorUtilityButton
   parent=SizeBox_1
   widget=EditorButton-Selection
   properties {
    toolTipText=Select actors added with Add to Scene
    widgetStyle {
     clickedSlateSound {resourceObject=null}
     disabled {
      bIsDynamicallyLoaded=false
      drawAs=Box
      imageType=FullColor
      mirroring=NoMirror
      resourceName=None
      resourceObject=&MI_PaintTools_button
      tiling=NoTile
      imageSize=<0 0>
      margin=<ltrb 0.48 0.48 0.48 0.48>
      outlineSettings {
       bUseBrushTransparency=false
       roundingType=FixedRadius
       width=1
       color {colorUseRule=UseColor_ColorTable, specifiedColor=<rgba 0 0 0 0>}
       cornerRadii=<4 4 4 4>
      }
      tintColor {colorUseRule=UseColor_Specified, specifiedColor=<rgba 0.0395 0.0395 0.0395 1>}
      uVRegion {bIsValid=false, max=<0 0>, min=<0 0>}
     }
     disabledForeground {colorUseRule=UseColor_ColorTable, specifiedColor=<rgba 0 0 0 0>}
     hovered {
      bIsDynamicallyLoaded=false
      drawAs=Box
      imageType=FullColor
      mirroring=NoMirror
      resourceName=None
      resourceObject=&MI_PaintTools_button
      tiling=NoTile
      imageSize=<32 32>
      margin=<ltrb 0.48 0.48 0.48 0.48>
      outlineSettings {
       bUseBrushTransparency=false
       roundingType=FixedRadius
       width=1
       color {colorUseRule=UseColor_ColorTable, specifiedColor=<rgba 0 0 0 0>}
       cornerRadii=<4 4 4 4>
      }
      tintColor {colorUseRule=UseColor_Specified, specifiedColor=<rgba 1 0.6198 0.1014 1>}
      uVRegion {bIsValid=false, max=<0 0>, min=<0 0>}
     }
     hoveredForeground {colorUseRule=UseColor_ColorTable, specifiedColor=<rgba 0 0 0 0>}
     hoveredSlateSound {resourceObject=null}
     normal {
      bIsDynamicallyLoaded=false
      drawAs=Box
      imageType=FullColor
      mirroring=NoMirror
      resourceName=None
      resourceObject=&MI_PaintTools_button
      tiling=NoTile
      imageSize=<32 32>
      margin=<ltrb 0.48 0.48 0.48 0.48>
      outlineSettings {
       bUseBrushTransparency=false
       roundingType=FixedRadius
       width=1
       color {colorUseRule=UseColor_ColorTable, specifiedColor=<rgba 0 0 0 0>}
       cornerRadii=<4 4 4 4>
      }
      tintColor {colorUseRule=UseColor_Specified, specifiedColor=<rgba 0.15 0.15 0.15 1>}
      uVRegion {bIsValid=false, max=<0 0>, min=<0 0>}
     }
     normalForeground {colorUseRule=UseColor_ColorTable, specifiedColor=<rgba 0 0 0 0>}
     normalPadding=<ltrb 12 1.5 12 1.5>
     pressed {
      bIsDynamicallyLoaded=false
      drawAs=Box
      imageType=FullColor
      mirroring=NoMirror
      resourceName=None
      resourceObject=&MI_PaintTools_button
      tiling=NoTile
      imageSize=<32 32>
      margin=<ltrb 0.48 0.48 0.48 0.48>
      outlineSettings {
       bUseBrushTransparency=false
       roundingType=FixedRadius
       width=1
       color {colorUseRule=UseColor_ColorTable, specifiedColor=<rgba 0 0 0 0>}
       cornerRadii=<4 4 4 4>
      }
      tintColor {colorUseRule=UseColor_Foreground, specifiedColor=<rgba 0.0395 0.0395 0.0395 1>}
      uVRegion {bIsValid=false, max=<0 0>, min=<0 0>}
     }
     pressedForeground {colorUseRule=UseColor_ColorTable, specifiedColor=<rgba 0 0 0 0>}
     pressedPadding=<ltrb 12 2.5 12 0.5>
     pressedSlateSound {resourceObject=null}
    }
   }
   slot {class=SizeBoxSlot}
  }
  img_selection {
   widgetClassPath=&Image
   bIsVariable=true
   parent=EditorButton-Selection
   widget=img_selection
   properties {
    pixelSnapping=SnapToPixel
    brush {
     bIsDynamicallyLoaded=false
     drawAs=Image
     imageType=FullColor
     mirroring=NoMirror
     resourceName=None
     resourceObject=&T_PaintTools_Select
     tiling=NoTile
     imageSize=<32 32>
     margin=<ltrb 0 0 0 0>
     outlineSettings {
      bUseBrushTransparency=false
      roundingType=HalfHeightRadius
      width=0
      color {colorUseRule=UseColor_Specified, specifiedColor=<rgba 0 0 0 0>}
      cornerRadii=<0 0 0 0>
     }
     tintColor {colorUseRule=UseColor_Specified, specifiedColor=<rgba 1 1 1 1>}
     uVRegion {bIsValid=false, max=<0 0>, min=<0 0>}
    }
    renderTransform {angle=0, scale=<2.8 0.9>, shear=<0 0>, translation=<0 0>}
   }
   slot {class=ButtonSlot, properties {padding=<ltrb 4 2 4 4>}}
  }
  SizeBox {
   widgetClassPath=&SizeBox
   parent=HorizontalBox_174
   widget=SizeBox
   properties {bOverride_HeightOverride=true, bOverride_WidthOverride=true, heightOverride=42, widthOverride=42}
   slot {
    class=HorizontalBoxSlot
    properties {horizontalAlignment=HAlign_Center, verticalAlignment=VAlign_Center, padding=<ltrb 4 0 0 0>, size {sizeRule=Fill, value=1}}
   }
  }
  EditorButton-Replace {
   widgetClassPath=&EditorUtilityButton
   parent=SizeBox
   widget=EditorButton-Replace
   properties {
    toolTipText=Replace with new selection from content browser
    widgetStyle {
     clickedSlateSound {resourceObject=null}
     disabled {
      bIsDynamicallyLoaded=false
      drawAs=Box
      imageType=FullColor
      mirroring=NoMirror
      resourceName=None
      resourceObject=&MI_PaintTools_button
      tiling=NoTile
      imageSize=<0 0>
      margin=<ltrb 0.48 0.48 0.48 0.48>
      outlineSettings {
       bUseBrushTransparency=false
       roundingType=FixedRadius
       width=1
       color {colorUseRule=UseColor_ColorTable, specifiedColor=<rgba 0 0 0 0>}
       cornerRadii=<4 4 4 4>
      }
      tintColor {colorUseRule=UseColor_Specified, specifiedColor=<rgba 0.0395 0.0395 0.0395 1>}
      uVRegion {bIsValid=false, max=<0 0>, min=<0 0>}
     }
     disabledForeground {colorUseRule=UseColor_ColorTable, specifiedColor=<rgba 0 0 0 0>}
     hovered {
      bIsDynamicallyLoaded=false
      drawAs=Box
      imageType=FullColor
      mirroring=NoMirror
      resourceName=None
      resourceObject=&MI_PaintTools_button
      tiling=NoTile
      imageSize=<32 32>
      margin=<ltrb 0.48 0.48 0.48 0.48>
      outlineSettings {
       bUseBrushTransparency=false
       roundingType=FixedRadius
       width=1
       color {colorUseRule=UseColor_ColorTable, specifiedColor=<rgba 0 0 0 0>}
       cornerRadii=<4 4 4 4>
      }
      tintColor {colorUseRule=UseColor_Specified, specifiedColor=<rgba 1 0.6198 0.1014 1>}
      uVRegion {bIsValid=false, max=<0 0>, min=<0 0>}
     }
     hoveredForeground {colorUseRule=UseColor_ColorTable, specifiedColor=<rgba 0 0 0 0>}
     hoveredSlateSound {resourceObject=null}
     normal {
      bIsDynamicallyLoaded=false
      drawAs=Box
      imageType=FullColor
      mirroring=NoMirror
      resourceName=None
      resourceObject=&MI_PaintTools_button
      tiling=NoTile
      imageSize=<32 32>
      margin=<ltrb 0.48 0.48 0.48 0.48>
      outlineSettings {
       bUseBrushTransparency=false
       roundingType=FixedRadius
       width=1
       color {colorUseRule=UseColor_ColorTable, specifiedColor=<rgba 0 0 0 0>}
       cornerRadii=<4 4 4 4>
      }
      tintColor {colorUseRule=UseColor_Specified, specifiedColor=<rgba 0.15 0.15 0.15 1>}
      uVRegion {bIsValid=false, max=<0 0>, min=<0 0>}
     }
     normalForeground {colorUseRule=UseColor_ColorTable, specifiedColor=<rgba 0 0 0 0>}
     normalPadding=<ltrb 12 1.5 12 1.5>
     pressed {
      bIsDynamicallyLoaded=false
      drawAs=Box
      imageType=FullColor
      mirroring=NoMirror
      resourceName=None
      resourceObject=&MI_PaintTools_button
      tiling=NoTile
      imageSize=<32 32>
      margin=<ltrb 0.48 0.48 0.48 0.48>
      outlineSettings {
       bUseBrushTransparency=false
       roundingType=FixedRadius
       width=1
       color {colorUseRule=UseColor_ColorTable, specifiedColor=<rgba 0 0 0 0>}
       cornerRadii=<4 4 4 4>
      }
      tintColor {colorUseRule=UseColor_Foreground, specifiedColor=<rgba 0.0395 0.0395 0.0395 1>}
      uVRegion {bIsValid=false, max=<0 0>, min=<0 0>}
     }
     pressedForeground {colorUseRule=UseColor_ColorTable, specifiedColor=<rgba 0 0 0 0>}
     pressedPadding=<ltrb 12 2.5 12 0.5>
     pressedSlateSound {resourceObject=null}
    }
   }
   slot {class=SizeBoxSlot}
  }
  img_replace {
   widgetClassPath=&Image
   bIsVariable=true
   parent=EditorButton-Replace
   widget=img_replace
   properties {
    pixelSnapping=SnapToPixel
    brush {
     bIsDynamicallyLoaded=false
     drawAs=Image
     imageType=FullColor
     mirroring=NoMirror
     resourceName=None
     resourceObject=&/T_PaintTools_Swap
     tiling=NoTile
     imageSize=<128 128>
     margin=<ltrb 0 0 0 0>
     outlineSettings {
      bUseBrushTransparency=false
      roundingType=HalfHeightRadius
      width=0
      color {colorUseRule=UseColor_Specified, specifiedColor=<rgba 0 0 0 0>}
      cornerRadii=<0 0 0 0>
     }
     tintColor {colorUseRule=UseColor_Specified, specifiedColor=<rgba 1 1 1 1>}
     uVRegion {bIsValid=false, max=<0 0>, min=<0 0>}
    }
    renderTransform {angle=0, scale=<2.8 0.9>, shear=<0 0>, translation=<0 0>}
   }
   slot {class=ButtonSlot, properties {padding=<ltrb 4 2 4 4>}}
  }
  SizeBox_2 {
   widgetClassPath=&SizeBox
   parent=HorizontalBox_174
   widget=SizeBox_2
   properties {bOverride_HeightOverride=true, bOverride_WidthOverride=true, heightOverride=42, widthOverride=42}
   slot {
    class=HorizontalBoxSlot
    properties {horizontalAlignment=HAlign_Center, verticalAlignment=VAlign_Center, padding=<ltrb 4 0 0 0>, size {sizeRule=Fill, value=1}}
   }
  }
  ExportFBXButton {
   widgetClassPath=&EditorUtilityButton
   parent=SizeBox_2
   widget=ExportFBXButton
   properties {
    toolTipText=Export the Selected actors to FBX
    widgetStyle {
     clickedSlateSound {resourceObject=null}
     disabled {
      bIsDynamicallyLoaded=false
      drawAs=Box
      imageType=FullColor
      mirroring=NoMirror
      resourceName=None
      resourceObject=&MI_PaintTools_button
      tiling=NoTile
      imageSize=<0 0>
      margin=<ltrb 0.48 0.48 0.48 0.48>
      outlineSettings {
       bUseBrushTransparency=false
       roundingType=FixedRadius
       width=1
       color {colorUseRule=UseColor_ColorTable, specifiedColor=<rgba 0 0 0 0>}
       cornerRadii=<4 4 4 4>
      }
      tintColor {colorUseRule=UseColor_Specified, specifiedColor=<rgba 0.0395 0.0395 0.0395 1>}
      uVRegion {bIsValid=false, max=<0 0>, min=<0 0>}
     }
     disabledForeground {colorUseRule=UseColor_ColorTable, specifiedColor=<rgba 0 0 0 0>}
     hovered {
      bIsDynamicallyLoaded=false
      drawAs=Box
      imageType=FullColor
      mirroring=NoMirror
      resourceName=None
      resourceObject=&MI_PaintTools_button
      tiling=NoTile
      imageSize=<32 32>
      margin=<ltrb 0.48 0.48 0.48 0.48>
      outlineSettings {
       bUseBrushTransparency=false
       roundingType=FixedRadius
       width=1
       color {colorUseRule=UseColor_ColorTable, specifiedColor=<rgba 0 0 0 0>}
       cornerRadii=<4 4 4 4>
      }
      tintColor {colorUseRule=UseColor_Specified, specifiedColor=<rgba 1 0.6198 0.1014 1>}
      uVRegion {bIsValid=false, max=<0 0>, min=<0 0>}
     }
     hoveredForeground {colorUseRule=UseColor_ColorTable, specifiedColor=<rgba 0 0 0 0>}
     hoveredSlateSound {resourceObject=null}
     normal {
      bIsDynamicallyLoaded=false
      drawAs=Box
      imageType=FullColor
      mirroring=NoMirror
      resourceName=None
      resourceObject=&MI_PaintTools_button
      tiling=NoTile
      imageSize=<32 32>
      margin=<ltrb 0.48 0.48 0.48 0.48>
      outlineSettings {
       bUseBrushTransparency=false
       roundingType=FixedRadius
       width=1
       color {colorUseRule=UseColor_ColorTable, specifiedColor=<rgba 0 0 0 0>}
       cornerRadii=<4 4 4 4>
      }
      tintColor {colorUseRule=UseColor_Specified, specifiedColor=<rgba 0.15 0.15 0.15 1>}
      uVRegion {bIsValid=false, max=<0 0>, min=<0 0>}
     }
     normalForeground {colorUseRule=UseColor_ColorTable, specifiedColor=<rgba 0 0 0 0>}
     normalPadding=<ltrb 12 1.5 12 1.5>
     pressed {
      bIsDynamicallyLoaded=false
      drawAs=Box
      imageType=FullColor
      mirroring=NoMirror
      resourceName=None
      resourceObject=&MI_PaintTools_button
      tiling=NoTile
      imageSize=<32 32>
      margin=<ltrb 0.48 0.48 0.48 0.48>
      outlineSettings {
       bUseBrushTransparency=false
       roundingType=FixedRadius
       width=1
       color {colorUseRule=UseColor_ColorTable, specifiedColor=<rgba 0 0 0 0>}
       cornerRadii=<4 4 4 4>
      }
      tintColor {colorUseRule=UseColor_Foreground, specifiedColor=<rgba 0.0395 0.0395 0.0395 1>}
      uVRegion {bIsValid=false, max=<0 0>, min=<0 0>}
     }
     pressedForeground {colorUseRule=UseColor_ColorTable, specifiedColor=<rgba 0 0 0 0>}
     pressedPadding=<ltrb 12 2.5 12 0.5>
     pressedSlateSound {resourceObject=null}
    }
   }
   slot {class=SizeBoxSlot}
  }
  img_fbx {
   widgetClassPath=&Image
   bIsVariable=true
   parent=ExportFBXButton
   widget=img_fbx
   properties {
    pixelSnapping=SnapToPixel
    brush {
     bIsDynamicallyLoaded=false
     drawAs=Image
     imageType=FullColor
     mirroring=NoMirror
     resourceName=None
     resourceObject=&/T_PaintTools_Export
     tiling=NoTile
     imageSize=<128 128>
     margin=<ltrb 0 0 0 0>
     outlineSettings {
      bUseBrushTransparency=false
      roundingType=HalfHeightRadius
      width=0
      color {colorUseRule=UseColor_Specified, specifiedColor=<rgba 0 0 0 0>}
      cornerRadii=<0 0 0 0>
     }
     tintColor {colorUseRule=UseColor_Specified, specifiedColor=<rgba 1 1 1 1>}
     uVRegion {bIsValid=false, max=<0 0>, min=<0 0>}
    }
    renderTransform {angle=0, scale=<2.8 0.9>, shear=<0 0>, translation=<0 0>}
   }
   slot {class=ButtonSlot, properties {padding=<ltrb 4 2 4 4>}}
  }
  HorizontalBox {
   widgetClassPath=&HorizontalBox
   parent=VerticalBox_214
   widget=HorizontalBox
   slot {class=VerticalBoxSlot, properties {horizontalAlignment=HAlign_Right, padding=<ltrb 0 -4 4 4>}}
  }
  SizeBox_477 {
   widgetClassPath=&SizeBox
   parent=HorizontalBox
   widget=SizeBox_477
   properties {bOverride_HeightOverride=true, bOverride_WidthOverride=true, heightOverride=42, widthOverride=42}
   slot {
    class=HorizontalBoxSlot
    properties {horizontalAlignment=HAlign_Center, verticalAlignment=VAlign_Center, padding=<ltrb 4 0 0 0>, size {sizeRule=Fill, value=1}}
   }
  }
  EditorButton-Delete {
   widgetClassPath=&EditorUtilityButton
   parent=SizeBox_477
   widget=EditorButton-Delete
   properties {
    toolTipText=Delete all Actors added with Add to Scene
    widgetStyle {
     clickedSlateSound {resourceObject=null}
     disabled {
      bIsDynamicallyLoaded=false
      drawAs=Box
      imageType=FullColor
      mirroring=NoMirror
      resourceName=None
      resourceObject=&MI_PaintTools_button
      tiling=NoTile
      imageSize=<0 0>
      margin=<ltrb 0.48 0.48 0.48 0.48>
      outlineSettings {
       bUseBrushTransparency=false
       roundingType=FixedRadius
       width=1
       color {colorUseRule=UseColor_ColorTable, specifiedColor=<rgba 0 0 0 0>}
       cornerRadii=<4 4 4 4>
      }
      tintColor {colorUseRule=UseColor_Specified, specifiedColor=<rgba 0.0395 0.0395 0.0395 1>}
      uVRegion {bIsValid=false, max=<0 0>, min=<0 0>}
     }
     disabledForeground {colorUseRule=UseColor_ColorTable, specifiedColor=<rgba 0 0 0 0>}
     hovered {
      bIsDynamicallyLoaded=false
      drawAs=Box
      imageType=FullColor
      mirroring=NoMirror
      resourceName=None
      resourceObject=&MI_PaintTools_button
      tiling=NoTile
      imageSize=<32 32>
      margin=<ltrb 0.48 0.48 0.48 0.48>
      outlineSettings {
       bUseBrushTransparency=false
       roundingType=FixedRadius
       width=1
       color {colorUseRule=UseColor_ColorTable, specifiedColor=<rgba 0 0 0 0>}
       cornerRadii=<4 4 4 4>
      }
      tintColor {colorUseRule=UseColor_Specified, specifiedColor=<rgba 0.276 0 0 1>}
      uVRegion {bIsValid=false, max=<0 0>, min=<0 0>}
     }
     hoveredForeground {colorUseRule=UseColor_Specified, specifiedColor=<rgba 1 1 1 0>}
     hoveredSlateSound {resourceObject=null}
     normal {
      bIsDynamicallyLoaded=false
      drawAs=Box
      imageType=FullColor
      mirroring=NoMirror
      resourceName=None
      resourceObject=&MI_PaintTools_button
      tiling=NoTile
      imageSize=<32 32>
      margin=<ltrb 0.48 0.48 0.48 0.48>
      outlineSettings {
       bUseBrushTransparency=false
       roundingType=FixedRadius
       width=1
       color {colorUseRule=UseColor_ColorTable, specifiedColor=<rgba 0 0 0 0>}
       cornerRadii=<4 4 4 4>
      }
      tintColor {colorUseRule=UseColor_Specified, specifiedColor=<rgba 0.2917 0.0698 0.0756 1>}
      uVRegion {bIsValid=false, max=<0 0>, min=<0 0>}
     }
     normalForeground {colorUseRule=UseColor_ColorTable, specifiedColor=<rgba 0 0 0 0>}
     normalPadding=<ltrb 12 1.5 12 1.5>
     pressed {
      bIsDynamicallyLoaded=false
      drawAs=Box
      imageType=FullColor
      mirroring=NoMirror
      resourceName=None
      resourceObject=&MI_PaintTools_button
      tiling=NoTile
      imageSize=<32 32>
      margin=<ltrb 0.48 0.48 0.48 0.48>
      outlineSettings {
       bUseBrushTransparency=false
       roundingType=FixedRadius
       width=1
       color {colorUseRule=UseColor_ColorTable, specifiedColor=<rgba 0 0 0 0>}
       cornerRadii=<4 4 4 4>
      }
      tintColor {colorUseRule=UseColor_Foreground, specifiedColor=<rgba 0.0395 0.0395 0.0395 1>}
      uVRegion {bIsValid=false, max=<0 0>, min=<0 0>}
     }
     pressedForeground {colorUseRule=UseColor_ColorTable, specifiedColor=<rgba 0 0 0 0>}
     pressedPadding=<ltrb 12 2.5 12 0.5>
     pressedSlateSound {resourceObject=null}
    }
   }
   slot {class=SizeBoxSlot}
  }
  img_delete {
   widgetClassPath=&Image
   bIsVariable=true
   parent=EditorButton-Delete
   widget=img_delete
   properties {
    pixelSnapping=SnapToPixel
    brush {
     bIsDynamicallyLoaded=false
     drawAs=Image
     imageType=FullColor
     mirroring=NoMirror
     resourceName=None
     resourceObject=&/T_PaintTools_Trash
     tiling=NoTile
     imageSize=<128 128>
     margin=<ltrb 0 0 0 0>
     outlineSettings {
      bUseBrushTransparency=false
      roundingType=HalfHeightRadius
      width=0
      color {colorUseRule=UseColor_Specified, specifiedColor=<rgba 0 0 0 0>}
      cornerRadii=<0 0 0 0>
     }
     tintColor {colorUseRule=UseColor_Specified, specifiedColor=<rgba 1 0.25 0.25 1>}
     uVRegion {bIsValid=false, max=<0 0>, min=<0 0>}
    }
    renderTransform {angle=0, scale=<2.8 0.9>, shear=<0 0>, translation=<0 0>}
   }
   slot {class=ButtonSlot, properties {padding=<ltrb 4 2 4 4>}}
  }
  SizeBox_3 {
   widgetClassPath=&SizeBox
   parent=HorizontalBox
   widget=SizeBox_3
   properties {bOverride_HeightOverride=true, bOverride_WidthOverride=true, heightOverride=42, widthOverride=42}
   slot {
    class=HorizontalBoxSlot
    properties {horizontalAlignment=HAlign_Center, verticalAlignment=VAlign_Center, padding=<ltrb 4 0 0 0>, size {sizeRule=Fill, value=1}}
   }
  }
  EditorButton-CleanupTT {
   widgetClassPath=&EditorUtilityButton
   parent=SizeBox_3
   widget=EditorButton-CleanupTT
   properties {
    toolTipText=Clean up starter actors in Turntable
    widgetStyle {
     clickedSlateSound {resourceObject=null}
     disabled {
      bIsDynamicallyLoaded=false
      drawAs=Box
      imageType=FullColor
      mirroring=NoMirror
      resourceName=None
      resourceObject=&MI_PaintTools_button
      tiling=NoTile
      imageSize=<0 0>
      margin=<ltrb 0.48 0.48 0.48 0.48>
      outlineSettings {
       bUseBrushTransparency=false
       roundingType=FixedRadius
       width=1
       color {colorUseRule=UseColor_ColorTable, specifiedColor=<rgba 0 0 0 0>}
       cornerRadii=<4 4 4 4>
      }
      tintColor {colorUseRule=UseColor_Specified, specifiedColor=<rgba 0.0395 0.0395 0.0395 1>}
      uVRegion {bIsValid=false, max=<0 0>, min=<0 0>}
     }
     disabledForeground {colorUseRule=UseColor_ColorTable, specifiedColor=<rgba 0 0 0 0>}
     hovered {
      bIsDynamicallyLoaded=false
      drawAs=Box
      imageType=FullColor
      mirroring=NoMirror
      resourceName=None
      resourceObject=&MI_PaintTools_button
      tiling=NoTile
      imageSize=<32 32>
      margin=<ltrb 0.48 0.48 0.48 0.48>
      outlineSettings {
       bUseBrushTransparency=false
       roundingType=FixedRadius
       width=1
       color {colorUseRule=UseColor_ColorTable, specifiedColor=<rgba 0 0 0 0>}
       cornerRadii=<4 4 4 4>
      }
      tintColor {colorUseRule=UseColor_Specified, specifiedColor=<rgba 1 0.6198 0.1014 1>}
      uVRegion {bIsValid=false, max=<0 0>, min=<0 0>}
     }
     hoveredForeground {colorUseRule=UseColor_Specified, specifiedColor=<rgba 1 1 1 0>}
     hoveredSlateSound {resourceObject=null}
     normal {
      bIsDynamicallyLoaded=false
      drawAs=Box
      imageType=FullColor
      mirroring=NoMirror
      resourceName=None
      resourceObject=&MI_PaintTools_button
      tiling=NoTile
      imageSize=<32 32>
      margin=<ltrb 0.48 0.48 0.48 0.48>
      outlineSettings {
       bUseBrushTransparency=false
       roundingType=FixedRadius
       width=1
       color {colorUseRule=UseColor_ColorTable, specifiedColor=<rgba 0 0 0 0>}
       cornerRadii=<4 4 4 4>
      }
      tintColor {colorUseRule=UseColor_Specified, specifiedColor=<rgba 0.15 0.15 0.15 1>}
      uVRegion {bIsValid=false, max=<0 0>, min=<0 0>}
     }
     normalForeground {colorUseRule=UseColor_ColorTable, specifiedColor=<rgba 0 0 0 0>}
     normalPadding=<ltrb 12 1.5 12 1.5>
     pressed {
      bIsDynamicallyLoaded=false
      drawAs=Box
      imageType=FullColor
      mirroring=NoMirror
      resourceName=None
      resourceObject=&MI_PaintTools_button
      tiling=NoTile
      imageSize=<32 32>
      margin=<ltrb 0.48 0.48 0.48 0.48>
      outlineSettings {
       bUseBrushTransparency=false
       roundingType=FixedRadius
       width=1
       color {colorUseRule=UseColor_ColorTable, specifiedColor=<rgba 0 0 0 0>}
       cornerRadii=<4 4 4 4>
      }
      tintColor {colorUseRule=UseColor_Foreground, specifiedColor=<rgba 0.0395 0.0395 0.0395 1>}
      uVRegion {bIsValid=false, max=<0 0>, min=<0 0>}
     }
     pressedForeground {colorUseRule=UseColor_ColorTable, specifiedColor=<rgba 0 0 0 0>}
     pressedPadding=<ltrb 12 2.5 12 0.5>
     pressedSlateSound {resourceObject=null}
    }
   }
   slot {class=SizeBoxSlot}
  }
  img_clean {
   widgetClassPath=&Image
   bIsVariable=true
   parent=EditorButton-CleanupTT
   widget=img_clean
   properties {
    pixelSnapping=SnapToPixel
    brush {
     bIsDynamicallyLoaded=false
     drawAs=Image
     imageType=FullColor
     mirroring=NoMirror
     resourceName=None
     resourceObject=&/T_PaintTools_Clean
     tiling=NoTile
     imageSize=<128 128>
     margin=<ltrb 0 0 0 0>
     outlineSettings {
      bUseBrushTransparency=false
      roundingType=HalfHeightRadius
      width=0
      color {colorUseRule=UseColor_Specified, specifiedColor=<rgba 0 0 0 0>}
      cornerRadii=<0 0 0 0>
     }
     tintColor {colorUseRule=UseColor_Specified, specifiedColor=<rgba 1 1 1 1>}
     uVRegion {bIsValid=false, max=<0 0>, min=<0 0>}
    }
    renderTransform {angle=0, scale=<2.8 0.9>, shear=<0 0>, translation=<0 0>}
   }
   slot {class=ButtonSlot, properties {padding=<ltrb 4 2 4 4>}}
  }
  Spacer_274 {widgetClassPath=&Spacer, parent=PANELVWRAP, widget=Spacer_274, properties {size=<1 4>}, slot {class=VerticalBoxSlot}}
  SETTINGS {widgetClassPath=&VerticalBox, parent=PANELVWRAP, widget=SETTINGS, slot {class=VerticalBoxSlot}}
  DetailsView_83 {
   widgetClassPath=&DetailsView
   parent=SETTINGS
   widget=DetailsView_83
   properties {visibility=Collapsed}
   slot {class=VerticalBoxSlot, properties {horizontalAlignment=HAlign_Center}}
  }
  Spacer_110 {widgetClassPath=&Spacer, parent=SETTINGS, widget=Spacer_110, properties {size=<1 8>}, slot {class=VerticalBoxSlot}}
  ParentSelector {
   widgetClassPath=&DetailsView
   parent=SETTINGS
   widget=ParentSelector
   properties {categoriesToShow [Actor], propertiesToShow [test]}
   slot {class=VerticalBoxSlot, properties {horizontalAlignment=HAlign_Center}}
  }
  Spacer {widgetClassPath=&Spacer, parent=SETTINGS, widget=Spacer, properties {size=<1 20>}, slot {class=VerticalBoxSlot}}
  INPUTVWRAPPER {
   widgetClassPath=&VerticalBox
   parent=SETTINGS
   widget=INPUTVWRAPPER
   slot {class=VerticalBoxSlot, properties {padding=<ltrb 20 0 20 0>}}
  }
  MESHSETTINGSHWRAPPER {
   widgetClassPath=&HorizontalBox
   parent=INPUTVWRAPPER
   widget=MESHSETTINGSHWRAPPER
   slot {class=VerticalBoxSlot, properties {padding=<ltrb 0 0 0 4>}}
  }
  LODCHECKBOX {
   widgetClassPath=&HorizontalBox
   parent=MESHSETTINGSHWRAPPER
   widget=LODCHECKBOX
   slot {class=HorizontalBoxSlot, properties {horizontalAlignment=HAlign_Left, padding=<ltrb 0 0 0 4>}}
  }
  checkboxMinLOD {
   widgetClassPath=&EditorUtilityCheckBox
   bIsVariable=true
   parent=LODCHECKBOX
   widget=checkboxMinLOD
   properties {toolTipText="When enabled, will set ForceMinLOD mesh property to 1.0, preventing LOD meshes from drawing in the viewport."}
   slot {class=HorizontalBoxSlot, properties {horizontalAlignment=HAlign_Left, verticalAlignment=VAlign_Center, padding=<ltrb 8 0 4 0>}}
  }
  TextMinLOD {
   widgetClassPath=&TextBlock
   bIsVariable=true
   parent=LODCHECKBOX
   widget=TextMinLOD
   properties {
    text="Forced Lod Model = 1.0"
    colorAndOpacity {colorUseRule=UseColor_Specified, specifiedColor=<rgba 0.85 0.85 0.85 1>}
    font {
     bForceMonospaced=false
     bMaterialIsStencil=false
     fontMaterial=null
     fontObject=&Roboto
     letterSpacing=0
     monospacedWidth=1
     size=9
     skewAmount=0
     typefaceFontName=Regular
     outlineSettings {
      bApplyOutlineToDropShadows=false
      bMiteredCorners=false
      bSeparateFillAlpha=false
      outlineMaterial=null
      outlineSize=0
      outlineColor=<rgba 0 0 0 1>
     }
    }
   }
   slot {class=HorizontalBoxSlot, properties {horizontalAlignment=HAlign_Left, verticalAlignment=VAlign_Center}}
  }
  grossspacer {widgetClassPath=&Spacer, parent=LODCHECKBOX, widget=grossspacer, properties {size=<16 1>}, slot {class=HorizontalBoxSlot}}
  Spacer_2 {widgetClassPath=&Spacer, parent=MESHSETTINGSHWRAPPER, widget=Spacer_2, properties {size=<16 1>}, slot {class=HorizontalBoxSlot}}
  MOVABLECHECKBOX {
   widgetClassPath=&HorizontalBox
   parent=MESHSETTINGSHWRAPPER
   widget=MOVABLECHECKBOX
   slot {class=HorizontalBoxSlot, properties {horizontalAlignment=HAlign_Left, padding=<ltrb 0 0 48 4>}}
  }
  checkboxMinLOD_1 {
   widgetClassPath=&EditorUtilityCheckBox
   bIsVariable=true
   parent=MOVABLECHECKBOX
   widget=checkboxMinLOD_1
   properties {toolTipText="When enabled, will set the movement type of the actor to MOVABLE, allowing it to be controlled by turntable controls."}
   slot {class=HorizontalBoxSlot, properties {horizontalAlignment=HAlign_Left, verticalAlignment=VAlign_Center, padding=<ltrb 8 0 4 0>}}
  }
  TextMovable {
   widgetClassPath=&TextBlock
   bIsVariable=true
   parent=MOVABLECHECKBOX
   widget=TextMovable
   properties {
    text=Set Movable
    colorAndOpacity {colorUseRule=UseColor_Specified, specifiedColor=<rgba 0.85 0.85 0.85 1>}
    font {
     bForceMonospaced=false
     bMaterialIsStencil=false
     fontMaterial=null
     fontObject=&Roboto
     letterSpacing=0
     monospacedWidth=1
     size=9
     skewAmount=0
     typefaceFontName=Regular
     outlineSettings {
      bApplyOutlineToDropShadows=false
      bMiteredCorners=false
      bSeparateFillAlpha=false
      outlineMaterial=null
      outlineSize=0
      outlineColor=<rgba 0 0 0 1>
     }
    }
   }
   slot {class=HorizontalBoxSlot, properties {horizontalAlignment=HAlign_Left, verticalAlignment=VAlign_Center}}
  }
  Spacer_1 {widgetClassPath=&Spacer, parent=MOVABLECHECKBOX, widget=Spacer_1, properties {size=<16 1>}, slot {class=HorizontalBoxSlot}}
  Image_98 {
   widgetClassPath=&Image
   bIsVariable=true
   parent=INPUTVWRAPPER
   widget=Image_98
   properties {brush {
     bIsDynamicallyLoaded=false
     drawAs=Image
     imageType=NoImage
     mirroring=NoMirror
     resourceName=None
     resourceObject=/InterchangeAssets/gltf/Textures/T_White_srgb
     tiling=NoTile
     imageSize=<1 1>
     margin=<ltrb 0 0 0 0>
     outlineSettings {
      bUseBrushTransparency=false
      roundingType=HalfHeightRadius
      width=0
      color {colorUseRule=UseColor_Specified, specifiedColor=<rgba 0 0 0 0>}
      cornerRadii=<0 0 0 0>
     }
     tintColor {colorUseRule=UseColor_Specified, specifiedColor=<rgba 0 0 0 1>}
     uVRegion {bIsValid=false, max=<0 0>, min=<0 0>}
    }}
   slot {class=VerticalBoxSlot, properties {verticalAlignment=VAlign_Center}}
  }
  Spacer_4 {widgetClassPath=&Spacer, parent=INPUTVWRAPPER, widget=Spacer_4, properties {size=<1 8>}, slot {class=VerticalBoxSlot}}
  ARRAYHWRAPPER {
   widgetClassPath=&HorizontalBox
   parent=INPUTVWRAPPER
   widget=ARRAYHWRAPPER
   slot {class=VerticalBoxSlot, properties {padding=<ltrb 0 0 0 8>}}
  }
  PLACEMENTCHECKBOX {
   widgetClassPath=&HorizontalBox
   parent=ARRAYHWRAPPER
   widget=PLACEMENTCHECKBOX
   slot {class=HorizontalBoxSlot, properties {horizontalAlignment=HAlign_Left, padding=<ltrb 0 0 0 4>}}
  }
  checkboxArrayAssets {
   widgetClassPath=&EditorUtilityCheckBox
   bIsVariable=true
   parent=PLACEMENTCHECKBOX
   widget=checkboxArrayAssets
   properties {toolTipText=Enable to postion in rows and columns. Use Spacing and Per Row settings to customize the layout.}
   slot {class=HorizontalBoxSlot, properties {horizontalAlignment=HAlign_Left, verticalAlignment=VAlign_Center, padding=<ltrb 8 0 4 0>}}
  }
  TextArrayAssets {
   widgetClassPath=&TextBlock
   bIsVariable=true
   parent=PLACEMENTCHECKBOX
   widget=TextArrayAssets
   properties {
    text=Array Placement
    colorAndOpacity {colorUseRule=UseColor_Specified, specifiedColor=<rgba 0.85 0.85 0.85 1>}
    font {
     bForceMonospaced=false
     bMaterialIsStencil=false
     fontMaterial=null
     fontObject=&Roboto
     letterSpacing=0
     monospacedWidth=1
     size=9
     skewAmount=0
     typefaceFontName=Regular
     outlineSettings {
      bApplyOutlineToDropShadows=false
      bMiteredCorners=false
      bSeparateFillAlpha=false
      outlineMaterial=null
      outlineSize=0
      outlineColor=<rgba 0 0 0 1>
     }
    }
   }
   slot {class=HorizontalBoxSlot, properties {horizontalAlignment=HAlign_Left, verticalAlignment=VAlign_Center}}
  }
  Spacer_3 {widgetClassPath=&Spacer, parent=ARRAYHWRAPPER, widget=Spacer_3, properties {size=<16 1>}, slot {class=HorizontalBoxSlot}}
  ARRAYBOXTEXTINPUT {
   widgetClassPath=&HorizontalBox
   parent=ARRAYHWRAPPER
   widget=ARRAYBOXTEXTINPUT
   slot {class=HorizontalBoxSlot, properties {horizontalAlignment=HAlign_Left, size {sizeRule=Fill, value=1}}}
  }
  HBOXSPACING {
   widgetClassPath=&HorizontalBox
   parent=ARRAYBOXTEXTINPUT
   widget=HBOXSPACING
   slot {class=HorizontalBoxSlot, properties {horizontalAlignment=HAlign_Right}}
  }
  TextSpacing {
   widgetClassPath=&TextBlock
   bIsVariable=true
   parent=HBOXSPACING
   widget=TextSpacing
   properties {
    text=Spacing
    colorAndOpacity {colorUseRule=UseColor_Specified, specifiedColor=<rgba 0.85 0.85 0.85 1>}
    font {
     bForceMonospaced=false
     bMaterialIsStencil=false
     fontMaterial=null
     fontObject=&Roboto
     letterSpacing=0
     monospacedWidth=1
     size=9
     skewAmount=0
     typefaceFontName=Regular
     outlineSettings {
      bApplyOutlineToDropShadows=false
      bMiteredCorners=false
      bSeparateFillAlpha=false
      outlineMaterial=null
      outlineSize=0
      outlineColor=<rgba 0 0 0 1>
     }
    }
   }
   slot {class=HorizontalBoxSlot, properties {horizontalAlignment=HAlign_Left, verticalAlignment=VAlign_Center}}
  }
  InputSpacing {
   widgetClassPath=&EditorUtilityEditableTextBox
   bIsVariable=true
   parent=HBOXSPACING
   widget=InputSpacing
   properties {
    hintText=Enter unit amount to space placed actors
    keyboardType=Number
    revertTextOnEscape=true
    selectAllTextWhenFocused=true
    text=256
    toolTipText="Input a value in Unreal Units as spacing. 512 is Default Fortnite Grid. "
   }
   slot {class=HorizontalBoxSlot, properties {padding=<ltrb 8 0 8 0>}}
  }
  HBOXARRAYSIZE {
   widgetClassPath=&HorizontalBox
   parent=ARRAYBOXTEXTINPUT
   widget=HBOXARRAYSIZE
   slot {class=HorizontalBoxSlot, properties {horizontalAlignment=HAlign_Right, padding=<ltrb 16 0 8 0>}}
  }
  TextPerRow {
   widgetClassPath=&TextBlock
   bIsVariable=true
   parent=HBOXARRAYSIZE
   widget=TextPerRow
   properties {
    text=Per Row
    colorAndOpacity {colorUseRule=UseColor_Specified, specifiedColor=<rgba 0.85 0.85 0.85 1>}
    font {
     bForceMonospaced=false
     bMaterialIsStencil=false
     fontMaterial=null
     fontObject=&Roboto
     letterSpacing=0
     monospacedWidth=1
     size=9
     skewAmount=0
     typefaceFontName=Regular
     outlineSettings {
      bApplyOutlineToDropShadows=false
      bMiteredCorners=false
      bSeparateFillAlpha=false
      outlineMaterial=null
      outlineSize=0
      outlineColor=<rgba 0 0 0 1>
     }
    }
   }
   slot {class=HorizontalBoxSlot, properties {horizontalAlignment=HAlign_Left, verticalAlignment=VAlign_Center}}
  }
  InuptRows {
   widgetClassPath=&EditorUtilityEditableTextBox
   bIsVariable=true
   parent=HBOXARRAYSIZE
   widget=InuptRows
   properties {
    hintText=Enter unit amount to space placed actors
    keyboardType=Number
    revertTextOnEscape=true
    selectAllTextWhenFocused=true
    text=8
    toolTipText="Number of meshes to be present per row. "
   }
   slot {class=HorizontalBoxSlot, properties {padding=<ltrb 8 0 8 0>}}
  }
  Image {
   widgetClassPath=&Image
   bIsVariable=true
   parent=INPUTVWRAPPER
   widget=Image
   properties {brush {
     bIsDynamicallyLoaded=false
     drawAs=Image
     imageType=NoImage
     mirroring=NoMirror
     resourceName=None
     resourceObject=&T_White_srgb
     tiling=NoTile
     imageSize=<1 1>
     margin=<ltrb 0 0 0 0>
     outlineSettings {
      bUseBrushTransparency=false
      roundingType=HalfHeightRadius
      width=0
      color {colorUseRule=UseColor_Specified, specifiedColor=<rgba 0 0 0 0>}
      cornerRadii=<0 0 0 0>
     }
     tintColor {colorUseRule=UseColor_Specified, specifiedColor=<rgba 0 0 0 1>}
     uVRegion {bIsValid=false, max=<0 0>, min=<0 0>}
    }}
   slot {class=VerticalBoxSlot, properties {verticalAlignment=VAlign_Center}}
  }
  Spacer_5 {widgetClassPath=&Spacer, parent=PANELVWRAP, widget=Spacer_5, properties {size=<1 8>}, slot {class=VerticalBoxSlot}}
  PaintTools {
   widgetClassPath=&VerticalBox
   parent=PANELVWRAP
   widget=PaintTools
   slot {class=VerticalBoxSlot, properties {padding=<ltrb 8 0 8 0>}}
  }
  HorizontalBox_110 {widgetClassPath=&HorizontalBox, parent=PaintTools, widget=HorizontalBox_110, slot {class=VerticalBoxSlot}}
  Buttons {
   widgetClassPath=&VerticalBox
   parent=HorizontalBox_110
   widget=Buttons
   slot {class=HorizontalBoxSlot, properties {padding=<ltrb 0 8 2 0>, size {sizeRule=Fill, value=1}}}
  }
  Header {widgetClassPath=&VerticalBox, parent=Buttons, widget=Header, slot {class=VerticalBoxSlot, properties {padding=<ltrb 0 0 0 8>}}}
  TextBlock_43 {
   widgetClassPath=&TextBlock
   parent=Header
   widget=TextBlock_43
   properties {
    text=Paint Tools
    font {
     bForceMonospaced=false
     bMaterialIsStencil=false
     fontMaterial=null
     fontObject=&Roboto
     letterSpacing=0
     monospacedWidth=1
     size=12
     skewAmount=0
     typefaceFontName=Bold
     outlineSettings {
      bApplyOutlineToDropShadows=false
      bMiteredCorners=false
      bSeparateFillAlpha=false
      outlineMaterial=null
      outlineSize=0
      outlineColor=<rgba 0 0 0 1>
     }
    }
   }
   slot {class=VerticalBoxSlot}
  }
  HR {
   widgetClassPath=&Border
   parent=Header
   widget=HR
   properties {
    background {
     bIsDynamicallyLoaded=false
     drawAs=Box
     imageType=NoImage
     mirroring=NoMirror
     resourceName=None
     resourceObject=null
     tiling=NoTile
     imageSize=<32 1>
     margin=<ltrb 0 0 0 0>
     outlineSettings {
      bUseBrushTransparency=false
      roundingType=HalfHeightRadius
      width=0
      color {colorUseRule=UseColor_Specified, specifiedColor=<rgba 0 0 0 0>}
      cornerRadii=<0 0 0 0>
     }
     tintColor {colorUseRule=UseColor_Specified, specifiedColor=<rgba 1 1 1 1>}
     uVRegion {bIsValid=false, max=<0 0>, min=<0 0>}
    }
    brushColor=<rgba 1 1 1 0.1>
    contentColorAndOpacity=<rgba 1 1 1 0.4609>
    desiredSizeScale=<1 0.5>
   }
   slot {class=VerticalBoxSlot}
  }
  EditorButton-ImportDemo {
   widgetClassPath=&EditorUtilityButton
   parent=Buttons
   widget=EditorButton-ImportDemo
   properties {
    toolTipText=DEMO mesh import Wizard will sync and import static meshes to the DEMO folder of the asset and attempt to assign materials.
    widgetStyle {
     clickedSlateSound {resourceObject=null}
     disabled {
      bIsDynamicallyLoaded=false
      drawAs=Box
      imageType=FullColor
      mirroring=NoMirror
      resourceName=None
      resourceObject=&/MI_PaintTools_TextButton
      tiling=NoTile
      imageSize=<0 0>
      margin=<ltrb 0.48 0.48 0.48 0.48>
      outlineSettings {
       bUseBrushTransparency=false
       roundingType=FixedRadius
       width=1
       color {colorUseRule=UseColor_ColorTable, specifiedColor=<rgba 0 0 0 0>}
       cornerRadii=<4 4 4 4>
      }
      tintColor {colorUseRule=UseColor_Specified, specifiedColor=<rgba 0.0395 0.0395 0.0395 1>}
      uVRegion {bIsValid=false, max=<0 0>, min=<0 0>}
     }
     disabledForeground {colorUseRule=UseColor_ColorTable, specifiedColor=<rgba 0 0 0 0>}
     hovered {
      bIsDynamicallyLoaded=false
      drawAs=Box
      imageType=FullColor
      mirroring=NoMirror
      resourceName=None
      resourceObject=&MI_PaintTools_TextButton
      tiling=NoTile
      imageSize=<32 32>
      margin=<ltrb 0.48 0.48 0.48 0.48>
      outlineSettings {
       bUseBrushTransparency=false
       roundingType=FixedRadius
       width=1
       color {colorUseRule=UseColor_ColorTable, specifiedColor=<rgba 0 0 0 0>}
       cornerRadii=<4 4 4 4>
      }
      tintColor {colorUseRule=UseColor_Specified, specifiedColor=<rgba 1 0.6198 0.1014 1>}
      uVRegion {bIsValid=false, max=<0 0>, min=<0 0>}
     }
     hoveredForeground {colorUseRule=UseColor_ColorTable, specifiedColor=<rgba 0 0 0 0>}
     hoveredSlateSound {resourceObject=null}
     normal {
      bIsDynamicallyLoaded=false
      drawAs=Box
      imageType=FullColor
      mirroring=NoMirror
      resourceName=None
      resourceObject=&MI_PaintTools_TextButton
      tiling=NoTile
      imageSize=<32 32>
      margin=<ltrb 0.48 0.48 0.48 0.48>
      outlineSettings {
       bUseBrushTransparency=false
       roundingType=FixedRadius
       width=1
       color {colorUseRule=UseColor_ColorTable, specifiedColor=<rgba 0 0 0 0>}
       cornerRadii=<4 4 4 4>
      }
      tintColor {colorUseRule=UseColor_Specified, specifiedColor=<rgba 0.15 0.15 0.15 1>}
      uVRegion {bIsValid=false, max=<0 0>, min=<0 0>}
     }
     normalForeground {colorUseRule=UseColor_ColorTable, specifiedColor=<rgba 0 0 0 0>}
     normalPadding=<ltrb 12 1.5 12 1.5>
     pressed {
      bIsDynamicallyLoaded=false
      drawAs=Box
      imageType=FullColor
      mirroring=NoMirror
      resourceName=None
      resourceObject=&MI_PaintTools_TextButton
      tiling=NoTile
      imageSize=<32 32>
      margin=<ltrb 0.48 0.48 0.48 0.48>
      outlineSettings {
       bUseBrushTransparency=false
       roundingType=FixedRadius
       width=1
       color {colorUseRule=UseColor_ColorTable, specifiedColor=<rgba 0 0 0 0>}
       cornerRadii=<4 4 4 4>
      }
      tintColor {colorUseRule=UseColor_Foreground, specifiedColor=<rgba 0.0395 0.0395 0.0395 1>}
      uVRegion {bIsValid=false, max=<0 0>, min=<0 0>}
     }
     pressedForeground {colorUseRule=UseColor_ColorTable, specifiedColor=<rgba 0 0 0 0>}
     pressedPadding=<ltrb 12 2.5 12 0.5>
     pressedSlateSound {resourceObject=null}
    }
   }
   slot {class=VerticalBoxSlot, properties {padding=<ltrb 0 0 0 2>}}
  }
  TextBlock_107 {
   widgetClassPath=&TextBlock
   parent=EditorButton-ImportDemo
   widget=TextBlock_107
   properties {
    text=Import DEMO Mesh
    font {
     bForceMonospaced=false
     bMaterialIsStencil=false
     fontMaterial=null
     fontObject=&Roboto
     letterSpacing=0
     monospacedWidth=1
     size=12
     skewAmount=0
     typefaceFontName=Regular
     outlineSettings {
      bApplyOutlineToDropShadows=false
      bMiteredCorners=false
      bSeparateFillAlpha=false
      outlineMaterial=null
      outlineSize=0
      outlineColor=<rgba 0 0 0 1>
     }
    }
   }
   slot {class=ButtonSlot, properties {horizontalAlignment=HAlign_Left, padding=<ltrb 4 4 4 4>}}
  }
  EditorButton-MaterialAssign {
   widgetClassPath=&EditorUtilityButton
   parent=Buttons
   widget=EditorButton-MaterialAssign
   properties {
    toolTipText=Attempt to assign selected materials to selected static meshes in Content Browser based on material slot names.
    widgetStyle {
     clickedSlateSound {resourceObject=null}
     disabled {
      bIsDynamicallyLoaded=false
      drawAs=Box
      imageType=FullColor
      mirroring=NoMirror
      resourceName=None
      resourceObject=&MI_PaintTools_TextButton
      tiling=NoTile
      imageSize=<0 0>
      margin=<ltrb 0.48 0.48 0.48 0.48>
      outlineSettings {
       bUseBrushTransparency=false
       roundingType=FixedRadius
       width=1
       color {colorUseRule=UseColor_ColorTable, specifiedColor=<rgba 0 0 0 0>}
       cornerRadii=<4 4 4 4>
      }
      tintColor {colorUseRule=UseColor_Specified, specifiedColor=<rgba 0.0395 0.0395 0.0395 1>}
      uVRegion {bIsValid=false, max=<0 0>, min=<0 0>}
     }
     disabledForeground {colorUseRule=UseColor_ColorTable, specifiedColor=<rgba 0 0 0 0>}
     hovered {
      bIsDynamicallyLoaded=false
      drawAs=Box
      imageType=FullColor
      mirroring=NoMirror
      resourceName=None
      resourceObject=&MI_PaintTools_TextButton
      tiling=NoTile
      imageSize=<32 32>
      margin=<ltrb 0.48 0.48 0.48 0.48>
      outlineSettings {
       bUseBrushTransparency=false
       roundingType=FixedRadius
       width=1
       color {colorUseRule=UseColor_ColorTable, specifiedColor=<rgba 0 0 0 0>}
       cornerRadii=<4 4 4 4>
      }
      tintColor {colorUseRule=UseColor_Specified, specifiedColor=<rgba 1 0.6198 0.1014 1>}
      uVRegion {bIsValid=false, max=<0 0>, min=<0 0>}
     }
     hoveredForeground {colorUseRule=UseColor_ColorTable, specifiedColor=<rgba 0 0 0 0>}
     hoveredSlateSound {resourceObject=null}
     normal {
      bIsDynamicallyLoaded=false
      drawAs=Box
      imageType=FullColor
      mirroring=NoMirror
      resourceName=None
      resourceObject=&MI_PaintTools_TextButton
      tiling=NoTile
      imageSize=<32 32>
      margin=<ltrb 0.48 0.48 0.48 0.48>
      outlineSettings {
       bUseBrushTransparency=false
       roundingType=FixedRadius
       width=1
       color {colorUseRule=UseColor_ColorTable, specifiedColor=<rgba 0 0 0 0>}
       cornerRadii=<4 4 4 4>
      }
      tintColor {colorUseRule=UseColor_Specified, specifiedColor=<rgba 0.15 0.15 0.15 1>}
      uVRegion {bIsValid=false, max=<0 0>, min=<0 0>}
     }
     normalForeground {colorUseRule=UseColor_ColorTable, specifiedColor=<rgba 0 0 0 0>}
     normalPadding=<ltrb 12 1.5 12 1.5>
     pressed {
      bIsDynamicallyLoaded=false
      drawAs=Box
      imageType=FullColor
      mirroring=NoMirror
      resourceName=None
      resourceObject=&MI_PaintTools_TextButton
      tiling=NoTile
      imageSize=<32 32>
      margin=<ltrb 0.48 0.48 0.48 0.48>
      outlineSettings {
       bUseBrushTransparency=false
       roundingType=FixedRadius
       width=1
       color {colorUseRule=UseColor_ColorTable, specifiedColor=<rgba 0 0 0 0>}
       cornerRadii=<4 4 4 4>
      }
      tintColor {colorUseRule=UseColor_Foreground, specifiedColor=<rgba 0.0395 0.0395 0.0395 1>}
      uVRegion {bIsValid=false, max=<0 0>, min=<0 0>}
     }
     pressedForeground {colorUseRule=UseColor_ColorTable, specifiedColor=<rgba 0 0 0 0>}
     pressedPadding=<ltrb 12 2.5 12 0.5>
     pressedSlateSound {resourceObject=null}
    }
   }
   slot {class=VerticalBoxSlot, properties {padding=<ltrb 0 0 0 2>}}
  }
  TextBlock {
   widgetClassPath=&TextBlock
   parent=EditorButton-MaterialAssign
   widget=TextBlock
   properties {
    text=Assign Materials To SM
    font {
     bForceMonospaced=false
     bMaterialIsStencil=false
     fontMaterial=null
     fontObject=&Roboto
     letterSpacing=0
     monospacedWidth=1
     size=12
     skewAmount=0
     typefaceFontName=Regular
     outlineSettings {
      bApplyOutlineToDropShadows=false
      bMiteredCorners=false
      bSeparateFillAlpha=false
      outlineMaterial=null
      outlineSize=0
      outlineColor=<rgba 0 0 0 1>
     }
    }
   }
   slot {class=ButtonSlot, properties {horizontalAlignment=HAlign_Left, padding=<ltrb 4 4 4 4>}}
  }
  EditorButton-MedLOD0 {
   widgetClassPath=&EditorUtilityButton
   parent=Buttons
   widget=EditorButton-MedLOD0
   properties {
    toolTipText=Set selected static mesh LOD to 1
    widgetStyle {
     clickedSlateSound {resourceObject=null}
     disabled {
      bIsDynamicallyLoaded=false
      drawAs=Box
      imageType=FullColor
      mirroring=NoMirror
      resourceName=None
      resourceObject=&MI_PaintTools_TextButton
      tiling=NoTile
      imageSize=<0 0>
      margin=<ltrb 0.48 0.48 0.48 0.48>
      outlineSettings {
       bUseBrushTransparency=false
       roundingType=FixedRadius
       width=1
       color {colorUseRule=UseColor_ColorTable, specifiedColor=<rgba 0 0 0 0>}
       cornerRadii=<4 4 4 4>
      }
      tintColor {colorUseRule=UseColor_Specified, specifiedColor=<rgba 0.0395 0.0395 0.0395 1>}
      uVRegion {bIsValid=false, max=<0 0>, min=<0 0>}
     }
     disabledForeground {colorUseRule=UseColor_ColorTable, specifiedColor=<rgba 0 0 0 0>}
     hovered {
      bIsDynamicallyLoaded=false
      drawAs=Box
      imageType=FullColor
      mirroring=NoMirror
      resourceName=None
      resourceObject=&MI_PaintTools_TextButton
      tiling=NoTile
      imageSize=<32 32>
      margin=<ltrb 0.48 0.48 0.48 0.48>
      outlineSettings {
       bUseBrushTransparency=false
       roundingType=FixedRadius
       width=1
       color {colorUseRule=UseColor_ColorTable, specifiedColor=<rgba 0 0 0 0>}
       cornerRadii=<4 4 4 4>
      }
      tintColor {colorUseRule=UseColor_Specified, specifiedColor=<rgba 1 0.6198 0.1014 1>}
      uVRegion {bIsValid=false, max=<0 0>, min=<0 0>}
     }
     hoveredForeground {colorUseRule=UseColor_ColorTable, specifiedColor=<rgba 0 0 0 0>}
     hoveredSlateSound {resourceObject=null}
     normal {
      bIsDynamicallyLoaded=false
      drawAs=Box
      imageType=FullColor
      mirroring=NoMirror
      resourceName=None
      resourceObject=&MI_PaintTools_TextButton
      tiling=NoTile
      imageSize=<32 32>
      margin=<ltrb 0.48 0.48 0.48 0.48>
      outlineSettings {
       bUseBrushTransparency=false
       roundingType=FixedRadius
       width=1
       color {colorUseRule=UseColor_ColorTable, specifiedColor=<rgba 0 0 0 0>}
       cornerRadii=<4 4 4 4>
      }
      tintColor {colorUseRule=UseColor_Specified, specifiedColor=<rgba 0.15 0.15 0.15 1>}
      uVRegion {bIsValid=false, max=<0 0>, min=<0 0>}
     }
     normalForeground {colorUseRule=UseColor_ColorTable, specifiedColor=<rgba 0 0 0 0>}
     normalPadding=<ltrb 12 1.5 12 1.5>
     pressed {
      bIsDynamicallyLoaded=false
      drawAs=Box
      imageType=FullColor
      mirroring=NoMirror
      resourceName=None
      resourceObject=&MI_PaintTools_TextButton
      tiling=NoTile
      imageSize=<32 32>
      margin=<ltrb 0.48 0.48 0.48 0.48>
      outlineSettings {
       bUseBrushTransparency=false
       roundingType=FixedRadius
       width=1
       color {colorUseRule=UseColor_ColorTable, specifiedColor=<rgba 0 0 0 0>}
       cornerRadii=<4 4 4 4>
      }
      tintColor {colorUseRule=UseColor_Foreground, specifiedColor=<rgba 0.0395 0.0395 0.0395 1>}
      uVRegion {bIsValid=false, max=<0 0>, min=<0 0>}
     }
     pressedForeground {colorUseRule=UseColor_ColorTable, specifiedColor=<rgba 0 0 0 0>}
     pressedPadding=<ltrb 12 2.5 12 0.5>
     pressedSlateSound {resourceObject=null}
    }
   }
   slot {class=VerticalBoxSlot, properties {padding=<ltrb 0 0 0 2>}}
  }
  TextBlock_7 {
   widgetClassPath=&TextBlock
   parent=EditorButton-MedLOD0
   widget=TextBlock_7
   properties {
    text=Set Mesh LOD
    font {
     bForceMonospaced=false
     bMaterialIsStencil=false
     fontMaterial=null
     fontObject=&Roboto
     letterSpacing=0
     monospacedWidth=1
     size=12
     skewAmount=0
     typefaceFontName=Regular
     outlineSettings {
      bApplyOutlineToDropShadows=false
      bMiteredCorners=false
      bSeparateFillAlpha=false
      outlineMaterial=null
      outlineSize=0
      outlineColor=<rgba 0 0 0 1>
     }
    }
   }
   slot {class=ButtonSlot, properties {horizontalAlignment=HAlign_Left, padding=<ltrb 4 4 4 4>}}
  }
  EditorButton-SkinToneAssist {
   widgetClassPath=&EditorUtilityButton
   parent=Buttons
   widget=EditorButton-SkinToneAssist
   properties {
    toolTipText="Set skin tint and subsurface profile on selected materials, based on exported data from Paint file."
    widgetStyle {
     clickedSlateSound {resourceObject=null}
     disabled {
      bIsDynamicallyLoaded=false
      drawAs=Box
      imageType=FullColor
      mirroring=NoMirror
      resourceName=None
      resourceObject=&MI_PaintTools_TextButton
      tiling=NoTile
      imageSize=<0 0>
      margin=<ltrb 0.48 0.48 0.48 0.48>
      outlineSettings {
       bUseBrushTransparency=false
       roundingType=FixedRadius
       width=1
       color {colorUseRule=UseColor_ColorTable, specifiedColor=<rgba 0 0 0 0>}
       cornerRadii=<4 4 4 4>
      }
      tintColor {colorUseRule=UseColor_Specified, specifiedColor=<rgba 0.0395 0.0395 0.0395 1>}
      uVRegion {bIsValid=false, max=<0 0>, min=<0 0>}
     }
     disabledForeground {colorUseRule=UseColor_ColorTable, specifiedColor=<rgba 0 0 0 0>}
     hovered {
      bIsDynamicallyLoaded=false
      drawAs=Box
      imageType=FullColor
      mirroring=NoMirror
      resourceName=None
      resourceObject=&MI_PaintTools_TextButton
      tiling=NoTile
      imageSize=<32 32>
      margin=<ltrb 0.48 0.48 0.48 0.48>
      outlineSettings {
       bUseBrushTransparency=false
       roundingType=FixedRadius
       width=1
       color {colorUseRule=UseColor_ColorTable, specifiedColor=<rgba 0 0 0 0>}
       cornerRadii=<4 4 4 4>
      }
      tintColor {colorUseRule=UseColor_Specified, specifiedColor=<rgba 1 0.6198 0.1014 1>}
      uVRegion {bIsValid=false, max=<0 0>, min=<0 0>}
     }
     hoveredForeground {colorUseRule=UseColor_ColorTable, specifiedColor=<rgba 0 0 0 0>}
     hoveredSlateSound {resourceObject=null}
     normal {
      bIsDynamicallyLoaded=false
      drawAs=Box
      imageType=FullColor
      mirroring=NoMirror
      resourceName=None
      resourceObject=&MI_PaintTools_TextButton
      tiling=NoTile
      imageSize=<32 32>
      margin=<ltrb 0.48 0.48 0.48 0.48>
      outlineSettings {
       bUseBrushTransparency=false
       roundingType=FixedRadius
       width=1
       color {colorUseRule=UseColor_ColorTable, specifiedColor=<rgba 0 0 0 0>}
       cornerRadii=<4 4 4 4>
      }
      tintColor {colorUseRule=UseColor_Specified, specifiedColor=<rgba 0.15 0.15 0.15 1>}
      uVRegion {bIsValid=false, max=<0 0>, min=<0 0>}
     }
     normalForeground {colorUseRule=UseColor_ColorTable, specifiedColor=<rgba 0 0 0 0>}
     normalPadding=<ltrb 12 1.5 12 1.5>
     pressed {
      bIsDynamicallyLoaded=false
      drawAs=Box
      imageType=FullColor
      mirroring=NoMirror
      resourceName=None
      resourceObject=&MI_PaintTools_TextButton
      tiling=NoTile
      imageSize=<32 32>
      margin=<ltrb 0.48 0.48 0.48 0.48>
      outlineSettings {
       bUseBrushTransparency=false
       roundingType=FixedRadius
       width=1
       color {colorUseRule=UseColor_ColorTable, specifiedColor=<rgba 0 0 0 0>}
       cornerRadii=<4 4 4 4>
      }
      tintColor {colorUseRule=UseColor_Foreground, specifiedColor=<rgba 0.0395 0.0395 0.0395 1>}
      uVRegion {bIsValid=false, max=<0 0>, min=<0 0>}
     }
     pressedForeground {colorUseRule=UseColor_ColorTable, specifiedColor=<rgba 0 0 0 0>}
     pressedPadding=<ltrb 12 2.5 12 0.5>
     pressedSlateSound {resourceObject=null}
    }
   }
   slot {class=VerticalBoxSlot, properties {padding=<ltrb 0 0 0 2>}}
  }
  TextBlock_1 {
   widgetClassPath=&TextBlock
   parent=EditorButton-SkinToneAssist
   widget=TextBlock_1
   properties {
    text=Skin Tone Assistant
    font {
     bForceMonospaced=false
     bMaterialIsStencil=false
     fontMaterial=null
     fontObject=&Roboto
     letterSpacing=0
     monospacedWidth=1
     size=12
     skewAmount=0
     typefaceFontName=Regular
     outlineSettings {
      bApplyOutlineToDropShadows=false
      bMiteredCorners=false
      bSeparateFillAlpha=false
      outlineMaterial=null
      outlineSize=0
      outlineColor=<rgba 0 0 0 1>
     }
    }
   }
   slot {class=ButtonSlot, properties {horizontalAlignment=HAlign_Left, padding=<ltrb 4 4 4 4>}}
  }
  EditorButton-EXT_MaterialCheatSheet {
   widgetClassPath=&EditorUtilityButton
   parent=Buttons
   widget=EditorButton-EXT_MaterialCheatSheet
   properties {
    toolTipText="Select uasset with  source in Content browser and copy depot path to asset. Paste path in p4 to navigate to source location."
    widgetStyle {
     clickedSlateSound {resourceObject=null}
     disabled {
      bIsDynamicallyLoaded=false
      drawAs=Box
      imageType=FullColor
      mirroring=NoMirror
      resourceName=None
      resourceObject=&MI_PaintTools_TextButton
      tiling=NoTile
      imageSize=<0 0>
      margin=<ltrb 0.48 0.48 0.48 0.48>
      outlineSettings {
       bUseBrushTransparency=false
       roundingType=FixedRadius
       width=1
       color {colorUseRule=UseColor_ColorTable, specifiedColor=<rgba 0 0 0 0>}
       cornerRadii=<4 4 4 4>
      }
      tintColor {colorUseRule=UseColor_Specified, specifiedColor=<rgba 0.0395 0.0395 0.0395 1>}
      uVRegion {bIsValid=false, max=<0 0>, min=<0 0>}
     }
     disabledForeground {colorUseRule=UseColor_ColorTable, specifiedColor=<rgba 0 0 0 0>}
     hovered {
      bIsDynamicallyLoaded=false
      drawAs=Box
      imageType=FullColor
      mirroring=NoMirror
      resourceName=None
      resourceObject=&MI_PaintTools_TextButton
      tiling=NoTile
      imageSize=<32 32>
      margin=<ltrb 0.48 0.48 0.48 0.48>
      outlineSettings {
       bUseBrushTransparency=false
       roundingType=FixedRadius
       width=1
       color {colorUseRule=UseColor_ColorTable, specifiedColor=<rgba 0 0 0 0>}
       cornerRadii=<4 4 4 4>
      }
      tintColor {colorUseRule=UseColor_Specified, specifiedColor=<rgba 1 0.6198 0.1014 1>}
      uVRegion {bIsValid=false, max=<0 0>, min=<0 0>}
     }
     hoveredForeground {colorUseRule=UseColor_ColorTable, specifiedColor=<rgba 0 0 0 0>}
     hoveredSlateSound {resourceObject=null}
     normal {
      bIsDynamicallyLoaded=false
      drawAs=Box
      imageType=FullColor
      mirroring=NoMirror
      resourceName=None
      resourceObject=&MI_PaintTools_TextButton
      tiling=NoTile
      imageSize=<32 32>
      margin=<ltrb 0.48 0.48 0.48 0.48>
      outlineSettings {
       bUseBrushTransparency=false
       roundingType=FixedRadius
       width=1
       color {colorUseRule=UseColor_ColorTable, specifiedColor=<rgba 0 0 0 0>}
       cornerRadii=<4 4 4 4>
      }
      tintColor {colorUseRule=UseColor_Specified, specifiedColor=<rgba 0.15 0.15 0.15 1>}
      uVRegion {bIsValid=false, max=<0 0>, min=<0 0>}
     }
     normalForeground {colorUseRule=UseColor_ColorTable, specifiedColor=<rgba 0 0 0 0>}
     normalPadding=<ltrb 12 1.5 12 1.5>
     pressed {
      bIsDynamicallyLoaded=false
      drawAs=Box
      imageType=FullColor
      mirroring=NoMirror
      resourceName=None
      resourceObject=&MI_PaintTools_TextButton
      tiling=NoTile
      imageSize=<32 32>
      margin=<ltrb 0.48 0.48 0.48 0.48>
      outlineSettings {
       bUseBrushTransparency=false
       roundingType=FixedRadius
       width=1
       color {colorUseRule=UseColor_ColorTable, specifiedColor=<rgba 0 0 0 0>}
       cornerRadii=<4 4 4 4>
      }
      tintColor {colorUseRule=UseColor_Foreground, specifiedColor=<rgba 0.0395 0.0395 0.0395 1>}
      uVRegion {bIsValid=false, max=<0 0>, min=<0 0>}
     }
     pressedForeground {colorUseRule=UseColor_ColorTable, specifiedColor=<rgba 0 0 0 0>}
     pressedPadding=<ltrb 12 2.5 12 0.5>
     pressedSlateSound {resourceObject=null}
    }
   }
   slot {class=VerticalBoxSlot, properties {padding=<ltrb 0 0 0 2>}}
  }
  TextBlock_9 {
   widgetClassPath=&TextBlock
   parent=EditorButton-EXT_MaterialCheatSheet
   widget=TextBlock_9
   properties {
    text=Material Cheat Sheet
    font {
     bForceMonospaced=false
     bMaterialIsStencil=false
     fontMaterial=null
     fontObject=&Roboto
     letterSpacing=0
     monospacedWidth=1
     size=12
     skewAmount=0
     typefaceFontName=Regular
     outlineSettings {
      bApplyOutlineToDropShadows=false
      bMiteredCorners=false
      bSeparateFillAlpha=false
      outlineMaterial=null
      outlineSize=0
      outlineColor=<rgba 0 0 0 1>
     }
    }
   }
   slot {class=ButtonSlot, properties {horizontalAlignment=HAlign_Left, padding=<ltrb 4 4 4 4>}}
  }
  Buttons_1 {
   widgetClassPath=&VerticalBox
   parent=HorizontalBox_110
   widget=Buttons_1
   slot {class=HorizontalBoxSlot, properties {padding=<ltrb 16 8 2 0>, size {sizeRule=Fill, value=1}}}
  }
  Header_1 {
   widgetClassPath=&VerticalBox
   parent=Buttons_1
   widget=Header_1
   slot {class=VerticalBoxSlot, properties {padding=<ltrb 0 0 0 8>}}
  }
  TextBlock_2 {
   widgetClassPath=&TextBlock
   parent=Header_1
   widget=TextBlock_2
   properties {
    text=" Shortcuts"
    font {
     bForceMonospaced=false
     bMaterialIsStencil=false
     fontMaterial=null
     fontObject=&Roboto
     letterSpacing=0
     monospacedWidth=1
     size=12
     skewAmount=0
     typefaceFontName=Bold
     outlineSettings {
      bApplyOutlineToDropShadows=false
      bMiteredCorners=false
      bSeparateFillAlpha=false
      outlineMaterial=null
      outlineSize=0
      outlineColor=<rgba 0 0 0 1>
     }
    }
   }
   slot {class=VerticalBoxSlot}
  }
  HR_1 {
   widgetClassPath=&Border
   parent=Header_1
   widget=HR_1
   properties {
    background {
     bIsDynamicallyLoaded=false
     drawAs=Box
     imageType=NoImage
     mirroring=NoMirror
     resourceName=None
     resourceObject=null
     tiling=NoTile
     imageSize=<32 1>
     margin=<ltrb 0 0 0 0>
     outlineSettings {
      bUseBrushTransparency=false
      roundingType=HalfHeightRadius
      width=0
      color {colorUseRule=UseColor_Specified, specifiedColor=<rgba 0 0 0 0>}
      cornerRadii=<0 0 0 0>
     }
     tintColor {colorUseRule=UseColor_Specified, specifiedColor=<rgba 1 1 1 1>}
     uVRegion {bIsValid=false, max=<0 0>, min=<0 0>}
    }
    brushColor=<rgba 1 1 1 0.1>
    contentColorAndOpacity=<rgba 1 1 1 0.4609>
    desiredSizeScale=<1 0.5>
   }
   slot {class=VerticalBoxSlot}
  }
  EditorButton-OpenGbufferTT {
   widgetClassPath=&EditorUtilityButton
   parent=Buttons_1
   widget=EditorButton-OpenGbufferTT
   properties {
    toolTipText=Quick launch Turntable_Gbuffers map.
    widgetStyle {
     clickedSlateSound {resourceObject=null}
     disabled {
      bIsDynamicallyLoaded=false
      drawAs=Box
      imageType=FullColor
      mirroring=NoMirror
      resourceName=None
      resourceObject=&MI_PaintTools_TextButton
      tiling=NoTile
      imageSize=<0 0>
      margin=<ltrb 0.48 0.48 0.48 0.48>
      outlineSettings {
       bUseBrushTransparency=false
       roundingType=FixedRadius
       width=1
       color {colorUseRule=UseColor_ColorTable, specifiedColor=<rgba 0 0 0 0>}
       cornerRadii=<4 4 4 4>
      }
      tintColor {colorUseRule=UseColor_Specified, specifiedColor=<rgba 0.0395 0.0395 0.0395 1>}
      uVRegion {bIsValid=false, max=<0 0>, min=<0 0>}
     }
     disabledForeground {colorUseRule=UseColor_ColorTable, specifiedColor=<rgba 0 0 0 0>}
     hovered {
      bIsDynamicallyLoaded=false
      drawAs=Box
      imageType=FullColor
      mirroring=NoMirror
      resourceName=None
      resourceObject=&MI_PaintTools_TextButton
      tiling=NoTile
      imageSize=<32 32>
      margin=<ltrb 0.48 0.48 0.48 0.48>
      outlineSettings {
       bUseBrushTransparency=false
       roundingType=FixedRadius
       width=1
       color {colorUseRule=UseColor_ColorTable, specifiedColor=<rgba 0 0 0 0>}
       cornerRadii=<4 4 4 4>
      }
      tintColor {colorUseRule=UseColor_Specified, specifiedColor=<rgba 1 0.6198 0.1014 1>}
      uVRegion {bIsValid=false, max=<0 0>, min=<0 0>}
     }
     hoveredForeground {colorUseRule=UseColor_ColorTable, specifiedColor=<rgba 0 0 0 0>}
     hoveredSlateSound {resourceObject=null}
     normal {
      bIsDynamicallyLoaded=false
      drawAs=Box
      imageType=FullColor
      mirroring=NoMirror
      resourceName=None
      resourceObject=&MI_PaintTools_TextButton
      tiling=NoTile
      imageSize=<32 32>
      margin=<ltrb 0.48 0.48 0.48 0.48>
      outlineSettings {
       bUseBrushTransparency=false
       roundingType=FixedRadius
       width=1
       color {colorUseRule=UseColor_ColorTable, specifiedColor=<rgba 0 0 0 0>}
       cornerRadii=<4 4 4 4>
      }
      tintColor {colorUseRule=UseColor_Specified, specifiedColor=<rgba 0.15 0.15 0.15 1>}
      uVRegion {bIsValid=false, max=<0 0>, min=<0 0>}
     }
     normalForeground {colorUseRule=UseColor_ColorTable, specifiedColor=<rgba 0 0 0 0>}
     normalPadding=<ltrb 12 1.5 12 1.5>
     pressed {
      bIsDynamicallyLoaded=false
      drawAs=Box
      imageType=FullColor
      mirroring=NoMirror
      resourceName=None
      resourceObject=&MI_PaintTools_TextButton
      tiling=NoTile
      imageSize=<32 32>
      margin=<ltrb 0.48 0.48 0.48 0.48>
      outlineSettings {
       bUseBrushTransparency=false
       roundingType=FixedRadius
       width=1
       color {colorUseRule=UseColor_ColorTable, specifiedColor=<rgba 0 0 0 0>}
       cornerRadii=<4 4 4 4>
      }
      tintColor {colorUseRule=UseColor_Foreground, specifiedColor=<rgba 0.0395 0.0395 0.0395 1>}
      uVRegion {bIsValid=false, max=<0 0>, min=<0 0>}
     }
     pressedForeground {colorUseRule=UseColor_ColorTable, specifiedColor=<rgba 0 0 0 0>}
     pressedPadding=<ltrb 12 2.5 12 0.5>
     pressedSlateSound {resourceObject=null}
    }
   }
   slot {class=VerticalBoxSlot, properties {padding=<ltrb 0 0 0 2>}}
  }
  TextBlock_3 {
   widgetClassPath=&TextBlock
   parent=EditorButton-OpenGbufferTT
   widget=TextBlock_3
   properties {
    text=Open GbufferTT
    font {
     bForceMonospaced=false
     bMaterialIsStencil=false
     fontMaterial=null
     fontObject=&Roboto
     letterSpacing=0
     monospacedWidth=1
     size=12
     skewAmount=0
     typefaceFontName=Regular
     outlineSettings {
      bApplyOutlineToDropShadows=false
      bMiteredCorners=false
      bSeparateFillAlpha=false
      outlineMaterial=null
      outlineSize=0
      outlineColor=<rgba 0 0 0 1>
     }
    }
   }
   slot {class=ButtonSlot, properties {horizontalAlignment=HAlign_Left, padding=<ltrb 4 4 4 4>}}
  }
  EditorButton-PilotGbufferCamera {
   widgetClassPath=&EditorUtilityButton
   parent=Buttons_1
   widget=EditorButton-PilotGbufferCamera
   properties {
    toolTipText=Toggle piloting camera in Turntable_gbuffer map
    widgetStyle {
     clickedSlateSound {resourceObject=null}
     disabled {
      bIsDynamicallyLoaded=false
      drawAs=Box
      imageType=FullColor
      mirroring=NoMirror
      resourceName=None
      resourceObject=&MI_PaintTools_TextButton
      tiling=NoTile
      imageSize=<0 0>
      margin=<ltrb 0.48 0.48 0.48 0.48>
      outlineSettings {
       bUseBrushTransparency=false
       roundingType=FixedRadius
       width=1
       color {colorUseRule=UseColor_ColorTable, specifiedColor=<rgba 0 0 0 0>}
       cornerRadii=<4 4 4 4>
      }
      tintColor {colorUseRule=UseColor_Specified, specifiedColor=<rgba 0.0395 0.0395 0.0395 1>}
      uVRegion {bIsValid=false, max=<0 0>, min=<0 0>}
     }
     disabledForeground {colorUseRule=UseColor_ColorTable, specifiedColor=<rgba 0 0 0 0>}
     hovered {
      bIsDynamicallyLoaded=false
      drawAs=Box
      imageType=FullColor
      mirroring=NoMirror
      resourceName=None
      resourceObject=&MI_PaintTools_TextButton
      tiling=NoTile
      imageSize=<32 32>
      margin=<ltrb 0.48 0.48 0.48 0.48>
      outlineSettings {
       bUseBrushTransparency=false
       roundingType=FixedRadius
       width=1
       color {colorUseRule=UseColor_ColorTable, specifiedColor=<rgba 0 0 0 0>}
       cornerRadii=<4 4 4 4>
      }
      tintColor {colorUseRule=UseColor_Specified, specifiedColor=<rgba 1 0.6198 0.1014 1>}
      uVRegion {bIsValid=false, max=<0 0>, min=<0 0>}
     }
     hoveredForeground {colorUseRule=UseColor_ColorTable, specifiedColor=<rgba 0 0 0 0>}
     hoveredSlateSound {resourceObject=null}
     normal {
      bIsDynamicallyLoaded=false
      drawAs=Box
      imageType=FullColor
      mirroring=NoMirror
      resourceName=None
      resourceObject=&MI_PaintTools_TextButton
      tiling=NoTile
      imageSize=<32 32>
      margin=<ltrb 0.48 0.48 0.48 0.48>
      outlineSettings {
       bUseBrushTransparency=false
       roundingType=FixedRadius
       width=1
       color {colorUseRule=UseColor_ColorTable, specifiedColor=<rgba 0 0 0 0>}
       cornerRadii=<4 4 4 4>
      }
      tintColor {colorUseRule=UseColor_Specified, specifiedColor=<rgba 0.15 0.15 0.15 1>}
      uVRegion {bIsValid=false, max=<0 0>, min=<0 0>}
     }
     normalForeground {colorUseRule=UseColor_ColorTable, specifiedColor=<rgba 0 0 0 0>}
     normalPadding=<ltrb 12 1.5 12 1.5>
     pressed {
      bIsDynamicallyLoaded=false
      drawAs=Box
      imageType=FullColor
      mirroring=NoMirror
      resourceName=None
      resourceObject=&MI_PaintTools_TextButton
      tiling=NoTile
      imageSize=<32 32>
      margin=<ltrb 0.48 0.48 0.48 0.48>
      outlineSettings {
       bUseBrushTransparency=false
       roundingType=FixedRadius
       width=1
       color {colorUseRule=UseColor_ColorTable, specifiedColor=<rgba 0 0 0 0>}
       cornerRadii=<4 4 4 4>
      }
      tintColor {colorUseRule=UseColor_Foreground, specifiedColor=<rgba 0.0395 0.0395 0.0395 1>}
      uVRegion {bIsValid=false, max=<0 0>, min=<0 0>}
     }
     pressedForeground {colorUseRule=UseColor_ColorTable, specifiedColor=<rgba 0 0 0 0>}
     pressedPadding=<ltrb 12 2.5 12 0.5>
     pressedSlateSound {resourceObject=null}
    }
   }
   slot {class=VerticalBoxSlot, properties {padding=<ltrb 0 0 0 2>}}
  }
  TextBlock_6 {
   widgetClassPath=&TextBlock
   parent=EditorButton-PilotGbufferCamera
   widget=TextBlock_6
   properties {
    text=Pilot TT Camera Toggle
    font {
     bForceMonospaced=false
     bMaterialIsStencil=false
     fontMaterial=null
     fontObject=&Roboto
     letterSpacing=0
     monospacedWidth=1
     size=12
     skewAmount=0
     typefaceFontName=Regular
     outlineSettings {
      bApplyOutlineToDropShadows=false
      bMiteredCorners=false
      bSeparateFillAlpha=false
      outlineMaterial=null
      outlineSize=0
      outlineColor=<rgba 0 0 0 1>
     }
    }
   }
   slot {class=ButtonSlot, properties {horizontalAlignment=HAlign_Left, padding=<ltrb 4 4 4 4>}}
  }
  EditorButton-CloseUpCamera {
   widgetClassPath=&EditorUtilityButton
   parent=Buttons_1
   widget=EditorButton-CloseUpCamera
   properties {
    toolTipText=Toggle Close Up Camera
    widgetStyle {
     clickedSlateSound {resourceObject=null}
     disabled {
      bIsDynamicallyLoaded=false
      drawAs=Box
      imageType=FullColor
      mirroring=NoMirror
      resourceName=None
      resourceObject=&MI_PaintTools_TextButton
      tiling=NoTile
      imageSize=<0 0>
      margin=<ltrb 0.48 0.48 0.48 0.48>
      outlineSettings {
       bUseBrushTransparency=false
       roundingType=FixedRadius
       width=1
       color {colorUseRule=UseColor_ColorTable, specifiedColor=<rgba 0 0 0 0>}
       cornerRadii=<4 4 4 4>
      }
      tintColor {colorUseRule=UseColor_Specified, specifiedColor=<rgba 0.0395 0.0395 0.0395 1>}
      uVRegion {bIsValid=false, max=<0 0>, min=<0 0>}
     }
     disabledForeground {colorUseRule=UseColor_ColorTable, specifiedColor=<rgba 0 0 0 0>}
     hovered {
      bIsDynamicallyLoaded=false
      drawAs=Box
      imageType=FullColor
      mirroring=NoMirror
      resourceName=None
      resourceObject=&MI_PaintTools_TextButton
      tiling=NoTile
      imageSize=<32 32>
      margin=<ltrb 0.48 0.48 0.48 0.48>
      outlineSettings {
       bUseBrushTransparency=false
       roundingType=FixedRadius
       width=1
       color {colorUseRule=UseColor_ColorTable, specifiedColor=<rgba 0 0 0 0>}
       cornerRadii=<4 4 4 4>
      }
      tintColor {colorUseRule=UseColor_Specified, specifiedColor=<rgba 1 0.6198 0.1014 1>}
      uVRegion {bIsValid=false, max=<0 0>, min=<0 0>}
     }
     hoveredForeground {colorUseRule=UseColor_ColorTable, specifiedColor=<rgba 0 0 0 0>}
     hoveredSlateSound {resourceObject=null}
     normal {
      bIsDynamicallyLoaded=false
      drawAs=Box
      imageType=FullColor
      mirroring=NoMirror
      resourceName=None
      resourceObject=&MI_PaintTools_TextButton
      tiling=NoTile
      imageSize=<32 32>
      margin=<ltrb 0.48 0.48 0.48 0.48>
      outlineSettings {
       bUseBrushTransparency=false
       roundingType=FixedRadius
       width=1
       color {colorUseRule=UseColor_ColorTable, specifiedColor=<rgba 0 0 0 0>}
       cornerRadii=<4 4 4 4>
      }
      tintColor {colorUseRule=UseColor_Specified, specifiedColor=<rgba 0.15 0.15 0.15 1>}
      uVRegion {bIsValid=false, max=<0 0>, min=<0 0>}
     }
     normalForeground {colorUseRule=UseColor_ColorTable, specifiedColor=<rgba 0 0 0 0>}
     normalPadding=<ltrb 12 1.5 12 1.5>
     pressed {
      bIsDynamicallyLoaded=false
      drawAs=Box
      imageType=FullColor
      mirroring=NoMirror
      resourceName=None
      resourceObject=&MI_PaintTools_TextButton
      tiling=NoTile
      imageSize=<32 32>
      margin=<ltrb 0.48 0.48 0.48 0.48>
      outlineSettings {
       bUseBrushTransparency=false
       roundingType=FixedRadius
       width=1
       color {colorUseRule=UseColor_ColorTable, specifiedColor=<rgba 0 0 0 0>}
       cornerRadii=<4 4 4 4>
      }
      tintColor {colorUseRule=UseColor_Foreground, specifiedColor=<rgba 0.0395 0.0395 0.0395 1>}
      uVRegion {bIsValid=false, max=<0 0>, min=<0 0>}
     }
     pressedForeground {colorUseRule=UseColor_ColorTable, specifiedColor=<rgba 0 0 0 0>}
     pressedPadding=<ltrb 12 2.5 12 0.5>
     pressedSlateSound {resourceObject=null}
    }
   }
   slot {class=VerticalBoxSlot, properties {padding=<ltrb 0 0 0 2>}}
  }
  CloseUpCamera {
   widgetClassPath=&TextBlock
   parent=EditorButton-CloseUpCamera
   widget=CloseUpCamera
   properties {
    text="Close Up TT Camera "
    font {
     bForceMonospaced=false
     bMaterialIsStencil=false
     fontMaterial=null
     fontObject=&Roboto
     letterSpacing=0
     monospacedWidth=1
     size=12
     skewAmount=0
     typefaceFontName=Regular
     outlineSettings {
      bApplyOutlineToDropShadows=false
      bMiteredCorners=false
      bSeparateFillAlpha=false
      outlineMaterial=null
      outlineSize=0
      outlineColor=<rgba 0 0 0 1>
     }
    }
   }
   slot {class=ButtonSlot, properties {horizontalAlignment=HAlign_Left, padding=<ltrb 4 4 4 4>}}
  }
  EditorButton-OpenBRArt {
   widgetClassPath=&EditorUtilityButton
   parent=Buttons_1
   widget=EditorButton-OpenBRArt
   properties {
    toolTipText=Quicklaunch current BR_Art map.
    widgetStyle {
     clickedSlateSound {resourceObject=null}
     disabled {
      bIsDynamicallyLoaded=false
      drawAs=Box
      imageType=FullColor
      mirroring=NoMirror
      resourceName=None
      resourceObject=&MI_PaintTools_TextButton
      tiling=NoTile
      imageSize=<0 0>
      margin=<ltrb 0.48 0.48 0.48 0.48>
      outlineSettings {
       bUseBrushTransparency=false
       roundingType=FixedRadius
       width=1
       color {colorUseRule=UseColor_ColorTable, specifiedColor=<rgba 0 0 0 0>}
       cornerRadii=<4 4 4 4>
      }
      tintColor {colorUseRule=UseColor_Specified, specifiedColor=<rgba 0.0395 0.0395 0.0395 1>}
      uVRegion {bIsValid=false, max=<0 0>, min=<0 0>}
     }
     disabledForeground {colorUseRule=UseColor_ColorTable, specifiedColor=<rgba 0 0 0 0>}
     hovered {
      bIsDynamicallyLoaded=false
      drawAs=Box
      imageType=FullColor
      mirroring=NoMirror
      resourceName=None
      resourceObject=&MI_PaintTools_TextButton
      tiling=NoTile
      imageSize=<32 32>
      margin=<ltrb 0.48 0.48 0.48 0.48>
      outlineSettings {
       bUseBrushTransparency=false
       roundingType=FixedRadius
       width=1
       color {colorUseRule=UseColor_ColorTable, specifiedColor=<rgba 0 0 0 0>}
       cornerRadii=<4 4 4 4>
      }
      tintColor {colorUseRule=UseColor_Specified, specifiedColor=<rgba 1 0.6198 0.1014 1>}
      uVRegion {bIsValid=false, max=<0 0>, min=<0 0>}
     }
     hoveredForeground {colorUseRule=UseColor_ColorTable, specifiedColor=<rgba 0 0 0 0>}
     hoveredSlateSound {resourceObject=null}
     normal {
      bIsDynamicallyLoaded=false
      drawAs=Box
      imageType=FullColor
      mirroring=NoMirror
      resourceName=None
      resourceObject=&MI_PaintTools_TextButton
      tiling=NoTile
      imageSize=<32 32>
      margin=<ltrb 0.48 0.48 0.48 0.48>
      outlineSettings {
       bUseBrushTransparency=false
       roundingType=FixedRadius
       width=1
       color {colorUseRule=UseColor_ColorTable, specifiedColor=<rgba 0 0 0 0>}
       cornerRadii=<4 4 4 4>
      }
      tintColor {colorUseRule=UseColor_Specified, specifiedColor=<rgba 0.15 0.15 0.15 1>}
      uVRegion {bIsValid=false, max=<0 0>, min=<0 0>}
     }
     normalForeground {colorUseRule=UseColor_ColorTable, specifiedColor=<rgba 0 0 0 0>}
     normalPadding=<ltrb 12 1.5 12 1.5>
     pressed {
      bIsDynamicallyLoaded=false
      drawAs=Box
      imageType=FullColor
      mirroring=NoMirror
      resourceName=None
      resourceObject=&MI_PaintTools_TextButton
      tiling=NoTile
      imageSize=<32 32>
      margin=<ltrb 0.48 0.48 0.48 0.48>
      outlineSettings {
       bUseBrushTransparency=false
       roundingType=FixedRadius
       width=1
       color {colorUseRule=UseColor_ColorTable, specifiedColor=<rgba 0 0 0 0>}
       cornerRadii=<4 4 4 4>
      }
      tintColor {colorUseRule=UseColor_Foreground, specifiedColor=<rgba 0.0395 0.0395 0.0395 1>}
      uVRegion {bIsValid=false, max=<0 0>, min=<0 0>}
     }
     pressedForeground {colorUseRule=UseColor_ColorTable, specifiedColor=<rgba 0 0 0 0>}
     pressedPadding=<ltrb 12 2.5 12 0.5>
     pressedSlateSound {resourceObject=null}
    }
   }
   slot {class=VerticalBoxSlot, properties {padding=<ltrb 0 0 0 2>}}
  }
  TextBlock_4 {
   widgetClassPath=&TextBlock
   parent=EditorButton-OpenBRArt
   widget=TextBlock_4
   properties {
    text=Open BR_Art
    font {
     bForceMonospaced=false
     bMaterialIsStencil=false
     fontMaterial=null
     fontObject=&Roboto
     letterSpacing=0
     monospacedWidth=1
     size=12
     skewAmount=0
     typefaceFontName=Regular
     outlineSettings {
      bApplyOutlineToDropShadows=false
      bMiteredCorners=false
      bSeparateFillAlpha=false
      outlineMaterial=null
      outlineSize=0
      outlineColor=<rgba 0 0 0 1>
     }
    }
   }
   slot {class=ButtonSlot, properties {horizontalAlignment=HAlign_Left, padding=<ltrb 4 4 4 4>}}
  }
  EditorButton-SpanwPaintWidgets {
   widgetClassPath=&EditorUtilityButton
   parent=Buttons_1
   widget=EditorButton-SpanwPaintWidgets
   properties {
    toolTipText="Quick access to common Paint tools, Cosmetic Picker, Turntable Widget Controls, Art Review Visualizer."
    widgetStyle {
     clickedSlateSound {resourceObject=null}
     disabled {
      bIsDynamicallyLoaded=false
      drawAs=Box
      imageType=FullColor
      mirroring=NoMirror
      resourceName=None
      resourceObject=&MI_PaintTools_TextButton
      tiling=NoTile
      imageSize=<0 0>
      margin=<ltrb 0.48 0.48 0.48 0.48>
      outlineSettings {
       bUseBrushTransparency=false
       roundingType=FixedRadius
       width=1
       color {colorUseRule=UseColor_ColorTable, specifiedColor=<rgba 0 0 0 0>}
       cornerRadii=<4 4 4 4>
      }
      tintColor {colorUseRule=UseColor_Specified, specifiedColor=<rgba 0.0395 0.0395 0.0395 1>}
      uVRegion {bIsValid=false, max=<0 0>, min=<0 0>}
     }
     disabledForeground {colorUseRule=UseColor_ColorTable, specifiedColor=<rgba 0 0 0 0>}
     hovered {
      bIsDynamicallyLoaded=false
      drawAs=Box
      imageType=FullColor
      mirroring=NoMirror
      resourceName=None
      resourceObject=&MI_PaintTools_TextButton
      tiling=NoTile
      imageSize=<32 32>
      margin=<ltrb 0.48 0.48 0.48 0.48>
      outlineSettings {
       bUseBrushTransparency=false
       roundingType=FixedRadius
       width=1
       color {colorUseRule=UseColor_ColorTable, specifiedColor=<rgba 0 0 0 0>}
       cornerRadii=<4 4 4 4>
      }
      tintColor {colorUseRule=UseColor_Specified, specifiedColor=<rgba 1 0.6198 0.1014 1>}
      uVRegion {bIsValid=false, max=<0 0>, min=<0 0>}
     }
     hoveredForeground {colorUseRule=UseColor_ColorTable, specifiedColor=<rgba 0 0 0 0>}
     hoveredSlateSound {resourceObject=null}
     normal {
      bIsDynamicallyLoaded=false
      drawAs=Box
      imageType=FullColor
      mirroring=NoMirror
      resourceName=None
      resourceObject=&MI_PaintTools_TextButton
      tiling=NoTile
      imageSize=<32 32>
      margin=<ltrb 0.48 0.48 0.48 0.48>
      outlineSettings {
       bUseBrushTransparency=false
       roundingType=FixedRadius
       width=1
       color {colorUseRule=UseColor_ColorTable, specifiedColor=<rgba 0 0 0 0>}
       cornerRadii=<4 4 4 4>
      }
      tintColor {colorUseRule=UseColor_Specified, specifiedColor=<rgba 0.15 0.15 0.15 1>}
      uVRegion {bIsValid=false, max=<0 0>, min=<0 0>}
     }
     normalForeground {colorUseRule=UseColor_ColorTable, specifiedColor=<rgba 0 0 0 0>}
     normalPadding=<ltrb 12 1.5 12 1.5>
     pressed {
      bIsDynamicallyLoaded=false
      drawAs=Box
      imageType=FullColor
      mirroring=NoMirror
      resourceName=None
      resourceObject=&MI_PaintTools_TextButton
      tiling=NoTile
      imageSize=<32 32>
      margin=<ltrb 0.48 0.48 0.48 0.48>
      outlineSettings {
       bUseBrushTransparency=false
       roundingType=FixedRadius
       width=1
       color {colorUseRule=UseColor_ColorTable, specifiedColor=<rgba 0 0 0 0>}
       cornerRadii=<4 4 4 4>
      }
      tintColor {colorUseRule=UseColor_Foreground, specifiedColor=<rgba 0.0395 0.0395 0.0395 1>}
      uVRegion {bIsValid=false, max=<0 0>, min=<0 0>}
     }
     pressedForeground {colorUseRule=UseColor_ColorTable, specifiedColor=<rgba 0 0 0 0>}
     pressedPadding=<ltrb 12 2.5 12 0.5>
     pressedSlateSound {resourceObject=null}
    }
   }
   slot {class=VerticalBoxSlot, properties {padding=<ltrb 0 0 0 2>}}
  }
  TextBlock_5 {
   widgetClassPath=&TextBlock
   parent=EditorButton-SpanwPaintWidgets
   widget=TextBlock_5
   properties {
    text=Open Paint Widgets
    font {
     bForceMonospaced=false
     bMaterialIsStencil=false
     fontMaterial=null
     fontObject=&Roboto
     letterSpacing=0
     monospacedWidth=1
     size=12
     skewAmount=0
     typefaceFontName=Regular
     outlineSettings {
      bApplyOutlineToDropShadows=false
      bMiteredCorners=false
      bSeparateFillAlpha=false
      outlineMaterial=null
      outlineSize=0
      outlineColor=<rgba 0 0 0 1>
     }
    }
   }
   slot {class=ButtonSlot, properties {horizontalAlignment=HAlign_Left, padding=<ltrb 4 4 4 4>}}
  }
  EditorButton-FormatMaterialString {
   widgetClassPath=&EditorUtilityButton
   parent=Buttons_1
   widget=EditorButton-FormatMaterialString
   properties {
    toolTipText="Quick access to common Paint tools, Cosmetic Picker, Turntable Widget Controls, Art Review Visualizer."
    widgetStyle {
     clickedSlateSound {resourceObject=null}
     disabled {
      bIsDynamicallyLoaded=false
      drawAs=Box
      imageType=FullColor
      mirroring=NoMirror
      resourceName=None
      resourceObject=&MI_PaintTools_TextButton
      tiling=NoTile
      imageSize=<0 0>
      margin=<ltrb 0.48 0.48 0.48 0.48>
      outlineSettings {
       bUseBrushTransparency=false
       roundingType=FixedRadius
       width=1
       color {colorUseRule=UseColor_ColorTable, specifiedColor=<rgba 0 0 0 0>}
       cornerRadii=<4 4 4 4>
      }
      tintColor {colorUseRule=UseColor_Specified, specifiedColor=<rgba 0.0395 0.0395 0.0395 1>}
      uVRegion {bIsValid=false, max=<0 0>, min=<0 0>}
     }
     disabledForeground {colorUseRule=UseColor_ColorTable, specifiedColor=<rgba 0 0 0 0>}
     hovered {
      bIsDynamicallyLoaded=false
      drawAs=Box
      imageType=FullColor
      mirroring=NoMirror
      resourceName=None
      resourceObject=&MI_PaintTools_TextButton
      tiling=NoTile
      imageSize=<32 32>
      margin=<ltrb 0.48 0.48 0.48 0.48>
      outlineSettings {
       bUseBrushTransparency=false
       roundingType=FixedRadius
       width=1
       color {colorUseRule=UseColor_ColorTable, specifiedColor=<rgba 0 0 0 0>}
       cornerRadii=<4 4 4 4>
      }
      tintColor {colorUseRule=UseColor_Specified, specifiedColor=<rgba 1 0.6198 0.1014 1>}
      uVRegion {bIsValid=false, max=<0 0>, min=<0 0>}
     }
     hoveredForeground {colorUseRule=UseColor_ColorTable, specifiedColor=<rgba 0 0 0 0>}
     hoveredSlateSound {resourceObject=null}
     normal {
      bIsDynamicallyLoaded=false
      drawAs=Box
      imageType=FullColor
      mirroring=NoMirror
      resourceName=None
      resourceObject=&MI_PaintTools_TextButton
      tiling=NoTile
      imageSize=<32 32>
      margin=<ltrb 0.48 0.48 0.48 0.48>
      outlineSettings {
       bUseBrushTransparency=false
       roundingType=FixedRadius
       width=1
       color {colorUseRule=UseColor_ColorTable, specifiedColor=<rgba 0 0 0 0>}
       cornerRadii=<4 4 4 4>
      }
      tintColor {colorUseRule=UseColor_Specified, specifiedColor=<rgba 0.15 0.15 0.15 1>}
      uVRegion {bIsValid=false, max=<0 0>, min=<0 0>}
     }
     normalForeground {colorUseRule=UseColor_ColorTable, specifiedColor=<rgba 0 0 0 0>}
     normalPadding=<ltrb 12 1.5 12 1.5>
     pressed {
      bIsDynamicallyLoaded=false
      drawAs=Box
      imageType=FullColor
      mirroring=NoMirror
      resourceName=None
      resourceObject=&MI_PaintTools_TextButton
      tiling=NoTile
      imageSize=<32 32>
      margin=<ltrb 0.48 0.48 0.48 0.48>
      outlineSettings {
       bUseBrushTransparency=false
       roundingType=FixedRadius
       width=1
       color {colorUseRule=UseColor_ColorTable, specifiedColor=<rgba 0 0 0 0>}
       cornerRadii=<4 4 4 4>
      }
      tintColor {colorUseRule=UseColor_Foreground, specifiedColor=<rgba 0.0395 0.0395 0.0395 1>}
      uVRegion {bIsValid=false, max=<0 0>, min=<0 0>}
     }
     pressedForeground {colorUseRule=UseColor_ColorTable, specifiedColor=<rgba 0 0 0 0>}
     pressedPadding=<ltrb 12 2.5 12 0.5>
     pressedSlateSound {resourceObject=null}
    }
   }
   slot {class=VerticalBoxSlot, properties {padding=<ltrb 0 0 0 2>}}
  }
  TextBlock_8 {
   widgetClassPath=&TextBlock
   parent=EditorButton-FormatMaterialString
   widget=TextBlock_8
   properties {
    text=Copy Mat Path to message
    font {
     bForceMonospaced=false
     bMaterialIsStencil=false
     fontMaterial=null
     fontObject=&Roboto
     letterSpacing=0
     monospacedWidth=1
     size=12
     skewAmount=0
     typefaceFontName=Regular
     outlineSettings {
      bApplyOutlineToDropShadows=false
      bMiteredCorners=false
      bSeparateFillAlpha=false
      outlineMaterial=null
      outlineSize=0
      outlineColor=<rgba 0 0 0 1>
     }
    }
   }
   slot {class=ButtonSlot, properties {horizontalAlignment=HAlign_Left, padding=<ltrb 4 4 4 4>}}
  }
 ]
}
