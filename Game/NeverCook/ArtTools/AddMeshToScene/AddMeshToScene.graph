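# Deprecated editor utility widget, formerly the Add Mesh To Scene art tool, now a redirect to its renamed replacement, Paint Tools.
# Its panel shows a deprecation notice, a summary of the new tool's features and a single button.
# Clicking the button spawns and registers the PaintTools widget tab through the EditorUtilitySubsystem.
# Many mesh placement, selection and turntable variables remain declared, but the graph uses none of them.
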EditorUtilityWidgetBlueprint /Game/NeverCook/ArtTools/AddMeshToScene/AddMeshToScene
graphs [
 EventGraph {nodes [
   "On Clicked (EditorButton-OPENPAINTWIDGET)"() {is=title K2Node_ComponentBoundEvent}
   > SpawnAndRegisterTab(self=Get EditorUtilitySubsystem(), InBlueprint=/Game/NeverCook/ArtTools/PaintTools/PaintTools)
  ]}
 NewEventDispatcher {nodes [NewEventDispatcher() {is=title K2Node_FunctionEntry}]}
]
variables [
 Actors to Select {type="array<Actor>"}
 Attach to Actor {type=Actor, default=null}
 bArrayAssets {type=boolean}
 bMinLOD {type=boolean, default=true}
 bMobilityMovable {type=boolean, default=true}
 character {type=AthenaCharacterItemDefinition, default=null}
 cleanUp {type="array<Actor>"}
 Cols {type=integer, default=8}
 counter {type=integer, default=1}
 disabledColorText {type=SlateColor, default {colorUseRule=UseColor_Specified, specifiedColor=<rgba 0.85 0.85 0.85 0.2>}}
 enabledColorText {type=SlateColor, default {colorUseRule=UseColor_Specified, specifiedColor=<rgba 0.85 0.85 0.85 1>}}
 MeshOverride_DV {type=DetailsView, default=null}
 MeshPosition {type=Vector}
 MRQSettingLocator {type=DetailsView, default=null}
 NewEventDispatcher {default=()}
 NewVar {type=boolean}
 nonParentTransform {type="array<Transform>", default []}
 OldTransform {type=Transform, default {location=<0 0 0>, rotation=<rot 0 0 0>, scale=<1 1 1>}}
 OutputFolder_DV {type=DetailsView, default=null}
 PerRow {type=integer, default=8}
 PlacementLocation {type=Vector}
 Rotation Slider Is Enabled {type=boolean, default=true}
 SelectedActors {type="array<Actor>"}
 SelectedActorsString {type="array<string>", default []}
 SelectionList {type="array<Object>", default []}
 Spacing {type=integer, default=512}
 TextBlock_126 {type=TextBlock, default=null}
 TurntableMap {type=string, default=Turntable_gbuffers}
 VersionInfo {type=GbufferTT_VersionInfo, default {current Version=2.4}}
 zeroTransform {type=Transform, default {location=<0 0 0>, rotation=<rot 0 0 0>, scale=<1 1 1>}}
]
widgets {
 info {rootWidgetClass=&StackBox, widgetCount=9}
 widgets [
  StackBox_0 {widgetClassPath=&StackBox, widget=StackBox_0}
  SizeBox_57 {
   widgetClassPath=&SizeBox
   parent=StackBox_0
   widget=SizeBox_57
   properties {bOverride_WidthOverride=true, widthOverride=480}
   slot {class=StackBoxSlot, properties {horizontalAlignment=HAlign_Center, verticalAlignment=VAlign_Top, size {sizeRule=Fill, value=1}}}
  }
  PANELVWRAP {
   widgetClassPath=&VerticalBox
   parent=SizeBox_57
   widget=PANELVWRAP
   slot {class=SizeBoxSlot, properties {verticalAlignment=VAlign_Center}}
  }
  MAINBUTTON {
   widgetClassPath=&VerticalBox
   parent=PANELVWRAP
   widget=MAINBUTTON
   slot {class=VerticalBoxSlot, properties {verticalAlignment=VAlign_Center}}
  }
  TextBlock_142 {
   widgetClassPath=&TextBlock
   parent=MAINBUTTON
   widget=TextBlock_142
   properties {
    justification=Center
    text="////  DEPRECATED ////\r\n RENAMED TO  PAINT TOOLS"
    colorAndOpacity {colorUseRule=UseColor_Specified, specifiedColor=<rgba 1 0.4117 0.15 1>}
    font {
     bForceMonospaced=false
     bMaterialIsStencil=false
     fontMaterial=null
     fontObject=&Roboto
     letterSpacing=0
     monospacedWidth=1
     size=18
     skewAmount=0
     typefaceFontName=Bold
     outlineSettings {
      bApplyOutlineToDropShadows=false
      bMiteredCorners=false
      bSeparateFillAlpha=false
      outlineMaterial=null
      outlineSize=0
      outlineColor=<rgba 0 0 0 1>
     }
    }
   }
   slot {class=VerticalBoxSlot, properties {horizontalAlignment=HAlign_Center, verticalAlignment=VAlign_Center}}
  }
  TextBlock_1 {
   widgetClassPath=&TextBlock
   parent=MAINBUTTON
   widget=TextBlock_1
   properties {
    text="Refactored widget to a new name with new features:\r\n\r\n- Demo mesh importing\r\n- one click  material assignment to meshes\r\n- Skin Tone assistant\r\n- Open GBuffers, BRArt one click\r\n- summon Paint Widgets"
    font {
     bForceMonospaced=false
     bMaterialIsStencil=false
     fontMaterial=null
     fontObject=&Roboto
     letterSpacing=0
     monospacedWidth=1
     size=10.5
     skewAmount=0
     typefaceFontName=Regular
     outlineSettings {
      bApplyOutlineToDropShadows=false
      bMiteredCorners=false
      bSeparateFillAlpha=false
      outlineMaterial=null
      outlineSize=0
      outlineColor=<rgba 0 0 0 1>
     }
    }
   }
   slot {class=VerticalBoxSlot, properties {horizontalAlignment=HAlign_Center, padding=<ltrb 0 16 0 16>}}
  }
  EditorButton-OPENPAINTWIDGET {
   widgetClassPath=&EditorUtilityButton
   parent=MAINBUTTON
   widget=EditorButton-OPENPAINTWIDGET
   properties {widgetStyle {
     clickedSlateSound {resourceObject=null}
     disabled {
      bIsDynamicallyLoaded=false
      drawAs=Box
      imageType=NoImage
      mirroring=NoMirror
      resourceName=None
      resourceObject=&/MI_AddMeshToScene_button
      tiling=NoTile
      imageSize=<0 0>
      margin=<ltrb 0.48 0.48 0.48 0.48>
      outlineSettings {
       bUseBrushTransparency=false
       roundingType=FixedRadius
       width=1
       color {colorUseRule=UseColor_ColorTable, specifiedColor=<rgba 0 0 0 0>}
       cornerRadii=<4 4 4 4>
      }
      tintColor {colorUseRule=UseColor_Specified, specifiedColor=<rgba 0.0395 0.0395 0.0395 1>}
      uVRegion {bIsValid=false, max=<0 0>, min=<0 0>}
     }
     disabledForeground {colorUseRule=UseColor_ColorTable, specifiedColor=<rgba 0 0 0 0>}
     hovered {
      bIsDynamicallyLoaded=false
      drawAs=Box
      imageType=NoImage
      mirroring=NoMirror
      resourceName=None
      resourceObject=&MI_AddMeshToScene_button
      tiling=NoTile
      imageSize=<32 32>
      margin=<ltrb 0.48 0.48 0.48 0.48>
      outlineSettings {
       bUseBrushTransparency=false
       roundingType=FixedRadius
       width=1
       color {colorUseRule=UseColor_ColorTable, specifiedColor=<rgba 0 0 0 0>}
       cornerRadii=<4 4 4 4>
      }
      tintColor {colorUseRule=UseColor_Specified, specifiedColor=<rgba 1 0.6198 0.1014 1>}
      uVRegion {bIsValid=false, max=<0 0>, min=<0 0>}
     }
     hoveredForeground {colorUseRule=UseColor_ColorTable, specifiedColor=<rgba 0 0 0 0>}
     hoveredSlateSound {resourceObject=null}
     normal {
      bIsDynamicallyLoaded=false
      drawAs=Box
      imageType=NoImage
      mirroring=NoMirror
      resourceName=None
      resourceObject=&MI_AddMeshToScene_button
      tiling=NoTile
      imageSize=<32 32>
      margin=<ltrb 0.48 0.48 0.48 0.48>
      outlineSettings {
       bUseBrushTransparency=false
       roundingType=FixedRadius
       width=1
       color {colorUseRule=UseColor_ColorTable, specifiedColor=<rgba 0 0 0 0>}
       cornerRadii=<4 4 4 4>
      }
      tintColor {colorUseRule=UseColor_Specified, specifiedColor=<rgba 0.3136 0.896 0 1>}
      uVRegion {bIsValid=false, max=<0 0>, min=<0 0>}
     }
     normalForeground {colorUseRule=UseColor_ColorTable, specifiedColor=<rgba 0 0 0 0>}
     normalPadding=<ltrb 12 1.5 12 1.5>
     pressed {
      bIsDynamicallyLoaded=false
      drawAs=Box
      imageType=NoImage
      mirroring=NoMirror
      resourceName=None
      resourceObject=&MI_AddMeshToScene_button
      tiling=NoTile
      imageSize=<32 32>
      margin=<ltrb 0.48 0.48 0.48 0.48>
      outlineSettings {
       bUseBrushTransparency=false
       roundingType=FixedRadius
       width=1
       color {colorUseRule=UseColor_ColorTable, specifiedColor=<rgba 0 0 0 0>}
       cornerRadii=<4 4 4 4>
      }
      tintColor {colorUseRule=UseColor_Foreground, specifiedColor=<rgba 0.0395 0.0395 0.0395 1>}
      uVRegion {bIsValid=false, max=<0 0>, min=<0 0>}
     }
     pressedForeground {colorUseRule=UseColor_ColorTable, specifiedColor=<rgba 0 0 0 0>}
     pressedPadding=<ltrb 12 2.5 12 0.5>
     pressedSlateSound {resourceObject=null}
    }}
   slot {class=VerticalBoxSlot, properties {verticalAlignment=VAlign_Center, padding=<ltrb 0 4 0 4>, size {sizeRule=Fill, value=1}}}
  }
  TextBlock {
   widgetClassPath=&TextBlock
   parent=EditorButton-OPENPAINTWIDGET
   widget=TextBlock
   properties {text=OPEN PAINT TOOLS}
   slot {class=ButtonSlot, properties {padding=<ltrb 4 20 4 24>}}
  }
  Spacer_274 {widgetClassPath=&Spacer, parent=PANELVWRAP, widget=Spacer_274, properties {size=<1 4>}, slot {class=VerticalBoxSlot}}
 ]
}
