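# Editor utility widget that spawns selected Content Browser assets into the level, attached to a parent actor, for the GBuffer turntable tool.
# On Construct, when a turntable exists, the parent defaults to an actor tagged TurnTable_UserObject.
# The button loops over the selection, optionally forcing LOD and Movable mobility on static and skeletal meshes.
# Array Placement arranges spawns in a grid using a counter, Spacing and Per Row, resetting after each batch.
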EditorUtilityWidgetBlueprint /Game/NeverCook/ArtTools/AddMeshToScene/addtosceneTT
graphs [
 EventGraph {nodes [
   "On Clicked (EditorUtilityButton_0)"() {is=title K2Node_ComponentBoundEvent}
   > n1 = GetSelectedAssets()
   > n2 = For Each Loop(Array:Array of Object References=@n1) -> (Array Element:Object Reference)
   Set PlacementLocation(execute=@n2.Completed, PlacementLocation:Vector="0, 0, 0") -> (Output_Get:Vector)
   > Set counter(counter:int=1) -> (Output_Get:int=0)
   n3 = SpawnActorFromObject!(execute=@n2.LoopBody, ObjectToUse:Object Reference=@n2, Location:Vector="0, 0, 0", Rotation:Rotator, bTransient:bool=false) -> (then, ReturnValue:Actor Object Reference)
   > n4 = Cast To StaticMeshActor!(execute, Object=@n3) -> (then, CastFailed, AsStatic Mesh Actor:Static Mesh Actor Object Reference) {is=title K2Node_DynamicCast}
   > n5 = Branch(Condition=Get bMinLOD() -> (bMinLOD:bool))
   n6 = Get StaticMeshComponent(self=@n4)
   > SetForcedLodModel!(execute, self:Static Mesh Component Object Reference=@n6, NewForcedLodModel:int=1) -> (then)
   > n7 = Branch(Condition=Get bMobilityMovable() -> (bMobilityMovable:bool))
   n8 = SetMobility!(execute=[@n7.then, @n5.else], self:Scene Component Object Reference=@n6, NewMobility:EComponentMobility Enum=Movable) -> (then)
   n9 = Get Attach to Actor() -> (Attach to Actor:Actor Object Reference)
   n10 = Attach Actor To Actor!(
    execute=[@n8.then, @n31.then, @n7.else]
    self:Actor Object Reference=@n3
    ParentActor:Actor Object Reference=@n9
    SocketName:Name=None
    LocationRule:EAttachmentRule Enum=KeepRelative
    RotationRule:EAttachmentRule Enum=KeepRelative
    ScaleRule:EAttachmentRule Enum=KeepRelative
    bWeldSimulatedBodies:bool=TRUE
   ) -> (then, ReturnValue:bool)
   n11 = Branch(execute=[@n30.then, @n10.then], Condition=Get bArrayAssets() -> (bArrayAssets:bool))
   n12 = Cast To SkeletalMeshActor!(execute=@n4.CastFailed, Object=@n3) -> (then, CastFailed, AsSkeletal Mesh Actor:Skeletal Mesh Actor Object Reference) {is=title K2Node_DynamicCast}
   n13 = Get SkeletalMeshComponent!(self:Skeletal Mesh Actor Object Reference=@n12) -> (SkeletalMeshComponent:Skeletal Mesh Component Object Reference)
   Set Relative Location!(
    execute=@n11.then
    self:Scene Component Object Reference=[Get RootComponent!(self:Actor Object Reference=@n3) -> (RootComponent:Scene Component Object Reference), @n13, @n6]
    NewLocation:Vector=Get PlacementLocation() -> (PlacementLocation:Vector)
    bSweep:bool=false
    bTeleport:bool=false
   ) -> (then, SweepHitResult:Hit Result Structure) {is=title K2Node_CallFunction}
   > n14 = Branch(Condition="Equal (Integer)"(A=Get counter() -> (counter:int=0), B=Get PerRow() -> (PerRow:int=0)))
   n15 = BreakVector(InVec=Get PlacementLocation())
   n16 = Get Spacing() -> (Spacing:int=0)
   > Set PlacementLocation(PlacementLocation=MakeVector(Y=float - float(A=@n15.Y, B:int=@n16), Z=@n15.Z))
   > Set counter(counter=1) -> (Output_Get=0)
   n17 = BreakVector(InVec=Get PlacementLocation())
   n18 = Get Spacing() -> (Spacing=0)
   Set PlacementLocation(execute=@n14.else, PlacementLocation=MakeVector(X=float + float(A=@n17, B:int=@n18), Y=@n17.Y, Z=@n17.Z))
   > Set counter(counter=int + int(A=Get counter() -> (counter=0))) -> (Output_Get=0)
   "On Property Changed (ParentSelector)"!() -> (OutputDelegate, then, PropertyName:Name=None) {is=title K2Node_ComponentBoundEvent}
   n19 = Get Attach to Actor()
   > n20 = Is Valid(InputObject=@n19)
   PrintLog!(execute=@n20.Is Not Valid, Log:Text="LOCTEXT(\"bad\")", __WorldContext:Object Reference) -> (then)
   PrintLog(execute=@n20.Is Valid, Log="LOCTEXT(\"good\")")
   > Set Attach to Actor(Attach to Actor:Actor Object Reference=@n19) -> (Output_Get:Actor Object Reference)
   n21 = "On Check State Changed (checkboxArrayAssets)"!() -> (OutputDelegate, then, bIsChecked:bool) {is=title K2Node_ComponentBoundEvent}
   > Set bArrayAssets(bArrayAssets:bool=@n21.bIsChecked) -> (Output_Get:bool)
   n22 = "On Text Changed (InputSpacing)"!() -> (OutputDelegate, then, Text:"Text (by ref)") {is=title K2Node_ComponentBoundEvent}
   n23 = "To String (Text)"(InText=@n22.Text)
   > Branch(Condition=IsNumeric!(SourceString:String=@n23) -> (ReturnValue:bool))
   > Set Spacing(Spacing:int=String To Integer(InString=@n23)) -> (Output_Get:int=0)
   n24 = "On Text Changed (InuptRows)"!() -> (OutputDelegate, then, Text:"Text (by ref)") {is=title K2Node_ComponentBoundEvent}
   n25 = "To String (Text)"(InText=@n24.Text)
   > Branch(Condition=IsNumeric(SourceString=@n25))
   > Set PerRow(PerRow:int=String To Integer(InString=@n25)) -> (Output_Get:int=0)
   n26 = "On Check State Changed (checkboxMinLOD)"!() -> (OutputDelegate, then, bIsChecked:bool) {is=title K2Node_ComponentBoundEvent}
   > Set bMinLOD(bMinLOD:bool=@n26.bIsChecked) -> (Output_Get:bool)
   n27 = "On Check State Changed (checkboxMinLOD_1)"!() -> (OutputDelegate, then, bIsChecked:bool) {is=title K2Node_ComponentBoundEvent}
   > Set bMobilityMovable(bMobilityMovable:bool=@n27.bIsChecked) -> (Output_Get:bool)
   Event Construct()
   > SetObject(self=Get ParentSelector() -> (ParentSelector:Details View Object Reference), NewObject=Self-Reference())
   > SetObject(
    self=[
     Get OutputFolder_DV() -> (OutputFolder_DV:Details View Object Reference)
     Get MRQSettingLocator() -> (MRQSettingLocator:Details View Object Reference)
     Get MeshOverride_DV() -> (MeshOverride_DV:Details View Object Reference)
    ]
    NewObject=Self-Reference()
   )
   > n28 = GetAllActorsOfClass!(
    execute
    WorldContextObject:Object Reference
    ActorClass:Actor Class Reference=/Gbuffer_Turntable/FN_PublicTurntable_Assets/BP/BP_Turntable_GBuffers.BP_Turntable_GBuffers_C
   ) -> (then, OutActors:Array of BP Turntable GBuffers Object References)
   > Branch(Condition=Is Not Empty!(TargetArray:Array of BP Turntable GBuffers Object References=@n28) -> (ReturnValue:bool))
   > n29 = GetAllActorsWithTag!(execute, WorldContextObject:Object Reference, Tag:Name=TurnTable_UserObject) -> (then, OutActors:Array of Actor Object References)
   > Branch(Condition=Is Not Empty(TargetArray:Array of Actor Object References=@n29))
   > Set Attach to Actor(Attach to Actor="Get (a copy)"(Array:Array of Actor Object References=@n29) -> (Output:Actor Object Reference))
   Event PreConstruct()
   Event Tick()
   n30 = Attach Actor To Actor(
    execute=@n12.CastFailed
    self=@n3
    ParentActor=@n9
    SocketName=None
    LocationRule=KeepRelative
    RotationRule=KeepRelative
    ScaleRule=KeepRelative
    bWeldSimulatedBodies=TRUE
   )
   SetForcedLOD!(execute=@n32.then, self:Skinned Mesh Component Object Reference=@n13, InNewForcedLOD:int=1) -> (then)
   > Branch(Condition=Get bMobilityMovable())
   > n31 = SetMobility(self=@n13, NewMobility=Movable)
   n32 = Branch(execute=@n12.then, Condition=Get bMinLOD())
  ]}
 NewEventDispatcher {nodes [NewEventDispatcher() {is=title K2Node_FunctionEntry}]}
 StateVisuals {nodes [
   n1 = StateVisuals!() -> (then, Condition:bool=~) {is=title K2Node_FunctionEntry}
   > n2 = Branch(Condition=@n1)
   n3 = Get TextSpacing() -> (TextSpacing:Text Object Reference)
   > Set ColorAndOpacity!(
    execute
    ColorAndOpacity:Slate Color Structure=Get enabledColorText() -> (enabledColorText:Slate Color Structure)
    self:Text Object Reference=@n3
   ) -> (then, Output_Get:Slate Color Structure)
   Set ColorAndOpacity(execute=@n2.else, ColorAndOpacity=Get disabledColorText() -> (disabledColorText:Slate Color Structure), self=@n3)
   Get checkboxArrayAssets() -> (checkboxArrayAssets:Editor Utility Check Box Object Reference)
  ]}
 getSelected {nodes [
   getSelected() {is=title K2Node_FunctionEntry}
   > n1 = GetSelectedAssets()
   > n2 = For Each Loop(Array=@n1)
   > n3 = SpawnActorFromObject(ObjectToUse=@n2, Location="0, 0, 0", bTransient=false)
   > Attach Actor To Actor(
    self=@n3
    ParentActor=Get Attach to Actor()
    SocketName=None
    LocationRule=KeepRelative
    RotationRule=KeepRelative
    ScaleRule=KeepRelative
    bWeldSimulatedBodies=TRUE
   )
   > n4 = Cast To StaticMeshActor(Object=@n3)
   n5 = Get StaticMeshComponent(self=@n4)
   > SetForcedLodModel(self=@n5, NewForcedLodModel=1)
   > SetMobility(self=@n5, NewMobility=Movable)
   n6 = Cast To SkeletalMeshActor(execute=@n4.CastFailed, Object=@n3)
   > n7 = SpawnActorFromObject(ObjectToUse=@n6, Location="0, 0, 0", bTransient=false)
   > n8 = Cast To SkeletalMeshActor(Object=@n7)
   n9 = Get SkeletalMeshComponent(self=@n8)
   > SetForcedLOD(self=@n9, InNewForcedLOD=1)
   > SetMobility(self=@n9, NewMobility=Movable)
  ]}
]
variables [
 Attach to Actor {type=Actor, default=null}
 bArrayAssets {type=boolean}
 bMinLOD {type=boolean, default=true}
 bMobilityMovable {type=boolean, default=true}
 Cols {type=integer, default=8}
 counter {type=integer, default=1}
 disabledColorText {type=SlateColor, default {colorUseRule=UseColor_Specified, specifiedColor=<rgba 0.85 0.85 0.85 0.2>}}
 enabledColorText {type=SlateColor, default {colorUseRule=UseColor_Specified, specifiedColor=<rgba 0.85 0.85 0.85 1>}}
 MeshOverride_DV {type=DetailsView, default=null}
 MRQSettingLocator {type=DetailsView, default=null}
 NewEventDispatcher {default=()}
 OutputFolder_DV {type=DetailsView, default=null}
 PerRow {type=integer, default=8}
 PlacementLocation {type=Vector}
 Rotation Slider Is Enabled {type=boolean, default=true}
 Spacing {type=integer, default=512}
 TextBlock_126 {type=TextBlock, default=null}
 VersionInfo {type=GbufferTT_VersionInfo, default {current Version=2.4}}
]
widgets {
 info {rootWidgetClass=&StackBox, widgetCount=38}
 widgets [
  StackBox_0 {widgetClassPath=&StackBox, widget=StackBox_0}
  SizeBox_57 {
   widgetClassPath=&SizeBox
   parent=StackBox_0
   widget=SizeBox_57
   properties {bOverride_WidthOverride=true, widthOverride=480}
   slot {class=StackBoxSlot, properties {horizontalAlignment=HAlign_Center, verticalAlignment=VAlign_Top, size {sizeRule=Fill, value=1}}}
  }
  PANELVWRAP {
   widgetClassPath=&VerticalBox
   parent=SizeBox_57
   widget=PANELVWRAP
   slot {class=SizeBoxSlot, properties {verticalAlignment=VAlign_Center}}
  }
  MAINBUTTON {
   widgetClassPath=&VerticalBox
   parent=PANELVWRAP
   widget=MAINBUTTON
   slot {class=VerticalBoxSlot, properties {verticalAlignment=VAlign_Center, size {sizeRule=Fill, value=1}}}
  }
  EditorUtilityButton_0 {
   widgetClassPath=&EditorUtilityButton
   parent=MAINBUTTON
   widget=EditorUtilityButton_0
   properties {widgetStyle {
     clickedSlateSound {resourceObject=null}
     disabled {
      bIsDynamicallyLoaded=false
      drawAs=RoundedBox
      imageType=NoImage
      mirroring=NoMirror
      resourceName=None
      resourceObject=&/MI_AddToScene_button
      tiling=NoTile
      imageSize=<0 0>
      margin=<ltrb 0 0 0 0>
      outlineSettings {
       bUseBrushTransparency=false
       roundingType=FixedRadius
       width=1
       color {colorUseRule=UseColor_ColorTable, specifiedColor=<rgba 0 0 0 0>}
       cornerRadii=<4 4 4 4>
      }
      tintColor {colorUseRule=UseColor_Specified, specifiedColor=<rgba 0.0395 0.0395 0.0395 1>}
      uVRegion {bIsValid=false, max=<0 0>, min=<0 0>}
     }
     disabledForeground {colorUseRule=UseColor_ColorTable, specifiedColor=<rgba 0 0 0 0>}
     hovered {
      bIsDynamicallyLoaded=false
      drawAs=Box
      imageType=NoImage
      mirroring=NoMirror
      resourceName=None
      resourceObject=&MI_AddToScene_button
      tiling=NoTile
      imageSize=<32 32>
      margin=<ltrb 0.48 0.48 0.48 0.48>
      outlineSettings {
       bUseBrushTransparency=false
       roundingType=FixedRadius
       width=1
       color {colorUseRule=UseColor_ColorTable, specifiedColor=<rgba 0 0 0 0>}
       cornerRadii=<4 4 4 4>
      }
      tintColor {colorUseRule=UseColor_Specified, specifiedColor=<rgba 1 0.6198 0.1014 1>}
      uVRegion {bIsValid=false, max=<0 0>, min=<0 0>}
     }
     hoveredForeground {colorUseRule=UseColor_ColorTable, specifiedColor=<rgba 0 0 0 0>}
     hoveredSlateSound {resourceObject=null}
     normal {
      bIsDynamicallyLoaded=false
      drawAs=Box
      imageType=NoImage
      mirroring=NoMirror
      resourceName=None
      resourceObject=&MI_AddToScene_button
      tiling=NoTile
      imageSize=<32 32>
      margin=<ltrb 0.48 0.48 0.48 0.48>
      outlineSettings {
       bUseBrushTransparency=false
       roundingType=FixedRadius
       width=1
       color {colorUseRule=UseColor_ColorTable, specifiedColor=<rgba 0 0 0 0>}
       cornerRadii=<4 4 4 4>
      }
      tintColor {colorUseRule=UseColor_Specified, specifiedColor=<rgba 0.15 0.15 0.15 1>}
      uVRegion {bIsValid=false, max=<0 0>, min=<0 0>}
     }
     normalForeground {colorUseRule=UseColor_ColorTable, specifiedColor=<rgba 0 0 0 0>}
     normalPadding=<ltrb 12 1.5 12 1.5>
     pressed {
      bIsDynamicallyLoaded=false
      drawAs=Box
      imageType=NoImage
      mirroring=NoMirror
      resourceName=None
      resourceObject=&MI_AddToScene_button
      tiling=NoTile
      imageSize=<32 32>
      margin=<ltrb 0.48 0.48 0.48 0.48>
      outlineSettings {
       bUseBrushTransparency=false
       roundingType=FixedRadius
       width=1
       color {colorUseRule=UseColor_ColorTable, specifiedColor=<rgba 0 0 0 0>}
       cornerRadii=<4 4 4 4>
      }
      tintColor {colorUseRule=UseColor_Foreground, specifiedColor=<rgba 0.0395 0.0395 0.0395 1>}
      uVRegion {bIsValid=false, max=<0 0>, min=<0 0>}
     }
     pressedForeground {colorUseRule=UseColor_ColorTable, specifiedColor=<rgba 0 0 0 0>}
     pressedPadding=<ltrb 12 2.5 12 0.5>
     pressedSlateSound {resourceObject=null}
    }}
   slot {class=VerticalBoxSlot, properties {verticalAlignment=VAlign_Center, padding=<ltrb 0 8 0 4>, size {sizeRule=Fill, value=1}}}
  }
  TextBlock_77 {
   widgetClassPath=&TextBlock
   parent=EditorUtilityButton_0
   widget=TextBlock_77
   properties {text=ADD TO SCENE}
   slot {class=ButtonSlot, properties {padding=<ltrb 4 20 4 24>}}
  }
  Spacer_274 {widgetClassPath=&Spacer, parent=PANELVWRAP, widget=Spacer_274, properties {size=<1 4>}, slot {class=VerticalBoxSlot}}
  SETTINGS {widgetClassPath=&VerticalBox, parent=PANELVWRAP, widget=SETTINGS, slot {class=VerticalBoxSlot}}
  DetailsView_83 {
   widgetClassPath=&DetailsView
   parent=SETTINGS
   widget=DetailsView_83
   properties {visibility=Collapsed}
   slot {class=VerticalBoxSlot, properties {horizontalAlignment=HAlign_Center}}
  }
  Spacer_110 {widgetClassPath=&Spacer, parent=SETTINGS, widget=Spacer_110, properties {size=<1 8>}, slot {class=VerticalBoxSlot}}
  ParentSelector {
   widgetClassPath=&DetailsView
   parent=SETTINGS
   widget=ParentSelector
   properties {categoriesToShow [Actor], propertiesToShow [test]}
   slot {class=VerticalBoxSlot, properties {horizontalAlignment=HAlign_Center}}
  }
  Spacer {widgetClassPath=&Spacer, parent=SETTINGS, widget=Spacer, properties {size=<1 20>}, slot {class=VerticalBoxSlot}}
  INPUTVWRAPPER {
   widgetClassPath=&VerticalBox
   parent=SETTINGS
   widget=INPUTVWRAPPER
   slot {class=VerticalBoxSlot, properties {padding=<ltrb 20 0 20 0>}}
  }
  MESHSETTINGSHWRAPPER {
   widgetClassPath=&HorizontalBox
   parent=INPUTVWRAPPER
   widget=MESHSETTINGSHWRAPPER
   slot {class=VerticalBoxSlot, properties {padding=<ltrb 0 0 0 4>}}
  }
  LODCHECKBOX {
   widgetClassPath=&HorizontalBox
   parent=MESHSETTINGSHWRAPPER
   widget=LODCHECKBOX
   slot {class=HorizontalBoxSlot, properties {horizontalAlignment=HAlign_Left, padding=<ltrb 0 0 0 4>}}
  }
  checkboxMinLOD {
   widgetClassPath=&EditorUtilityCheckBox
   bIsVariable=true
   parent=LODCHECKBOX
   widget=checkboxMinLOD
   properties {toolTipText="When enabled, will set ForceMinLOD mesh property to 1.0, preventing LOD meshes from drawing in the viewport."}
   slot {class=HorizontalBoxSlot, properties {horizontalAlignment=HAlign_Left, verticalAlignment=VAlign_Center, padding=<ltrb 8 0 4 0>}}
  }
  TextMinLOD {
   widgetClassPath=&TextBlock
   bIsVariable=true
   parent=LODCHECKBOX
   widget=TextMinLOD
   properties {
    text="Forced Lod Model = 1.0"
    colorAndOpacity {colorUseRule=UseColor_Specified, specifiedColor=<rgba 0.85 0.85 0.85 1>}
    font {
     bForceMonospaced=false
     bMaterialIsStencil=false
     fontMaterial=null
     fontObject=&Roboto
     letterSpacing=0
     monospacedWidth=1
     size=9
     skewAmount=0
     typefaceFontName=Regular
     outlineSettings {
      bApplyOutlineToDropShadows=false
      bMiteredCorners=false
      bSeparateFillAlpha=false
      outlineMaterial=null
      outlineSize=0
      outlineColor=<rgba 0 0 0 1>
     }
    }
   }
   slot {class=HorizontalBoxSlot, properties {horizontalAlignment=HAlign_Left, verticalAlignment=VAlign_Center}}
  }
  grossspacer {widgetClassPath=&Spacer, parent=LODCHECKBOX, widget=grossspacer, properties {size=<16 1>}, slot {class=HorizontalBoxSlot}}
  Spacer_2 {widgetClassPath=&Spacer, parent=MESHSETTINGSHWRAPPER, widget=Spacer_2, properties {size=<16 1>}, slot {class=HorizontalBoxSlot}}
  MOVABLECHECKBOX {
   widgetClassPath=&HorizontalBox
   parent=MESHSETTINGSHWRAPPER
   widget=MOVABLECHECKBOX
   slot {class=HorizontalBoxSlot, properties {horizontalAlignment=HAlign_Left, padding=<ltrb 0 0 48 4>}}
  }
  checkboxMinLOD_1 {
   widgetClassPath=&EditorUtilityCheckBox
   bIsVariable=true
   parent=MOVABLECHECKBOX
   widget=checkboxMinLOD_1
   properties {toolTipText="When enabled, will set the movement type of the actor to MOVABLE, allowing it to be controlled by turntable controls."}
   slot {class=HorizontalBoxSlot, properties {horizontalAlignment=HAlign_Left, verticalAlignment=VAlign_Center, padding=<ltrb 8 0 4 0>}}
  }
  TextMovable {
   widgetClassPath=&TextBlock
   bIsVariable=true
   parent=MOVABLECHECKBOX
   widget=TextMovable
   properties {
    text=Set Movable
    colorAndOpacity {colorUseRule=UseColor_Specified, specifiedColor=<rgba 0.85 0.85 0.85 1>}
    font {
     bForceMonospaced=false
     bMaterialIsStencil=false
     fontMaterial=null
     fontObject=&Roboto
     letterSpacing=0
     monospacedWidth=1
     size=9
     skewAmount=0
     typefaceFontName=Regular
     outlineSettings {
      bApplyOutlineToDropShadows=false
      bMiteredCorners=false
      bSeparateFillAlpha=false
      outlineMaterial=null
      outlineSize=0
      outlineColor=<rgba 0 0 0 1>
     }
    }
   }
   slot {class=HorizontalBoxSlot, properties {horizontalAlignment=HAlign_Left, verticalAlignment=VAlign_Center}}
  }
  Spacer_1 {widgetClassPath=&Spacer, parent=MOVABLECHECKBOX, widget=Spacer_1, properties {size=<16 1>}, slot {class=HorizontalBoxSlot}}
  Image_98 {
   widgetClassPath=&Image
   bIsVariable=true
   parent=INPUTVWRAPPER
   widget=Image_98
   properties {brush {
     bIsDynamicallyLoaded=false
     drawAs=Image
     imageType=NoImage
     mirroring=NoMirror
     resourceName=None
     resourceObject=/InterchangeAssets/gltf/Textures/T_White_srgb
     tiling=NoTile
     imageSize=<1 1>
     margin=<ltrb 0 0 0 0>
     outlineSettings {
      bUseBrushTransparency=false
      roundingType=HalfHeightRadius
      width=0
      color {colorUseRule=UseColor_Specified, specifiedColor=<rgba 0 0 0 0>}
      cornerRadii=<0 0 0 0>
     }
     tintColor {colorUseRule=UseColor_Specified, specifiedColor=<rgba 0 0 0 1>}
     uVRegion {bIsValid=false, max=<0 0>, min=<0 0>}
    }}
   slot {class=VerticalBoxSlot, properties {verticalAlignment=VAlign_Center}}
  }
  Spacer_4 {widgetClassPath=&Spacer, parent=INPUTVWRAPPER, widget=Spacer_4, properties {size=<1 8>}, slot {class=VerticalBoxSlot}}
  ARRAYHWRAPPER {
   widgetClassPath=&HorizontalBox
   parent=INPUTVWRAPPER
   widget=ARRAYHWRAPPER
   slot {class=VerticalBoxSlot, properties {padding=<ltrb 0 0 0 8>}}
  }
  PLACEMENTCHECKBOX {
   widgetClassPath=&HorizontalBox
   parent=ARRAYHWRAPPER
   widget=PLACEMENTCHECKBOX
   slot {class=HorizontalBoxSlot, properties {horizontalAlignment=HAlign_Left, padding=<ltrb 0 0 0 4>}}
  }
  checkboxArrayAssets {
   widgetClassPath=&EditorUtilityCheckBox
   bIsVariable=true
   parent=PLACEMENTCHECKBOX
   widget=checkboxArrayAssets
   properties {toolTipText=Enable to postion in rows and columns. Use Spacing and Per Row settings to customize the layout.}
   slot {class=HorizontalBoxSlot, properties {horizontalAlignment=HAlign_Left, verticalAlignment=VAlign_Center, padding=<ltrb 8 0 4 0>}}
  }
  TextArrayAssets {
   widgetClassPath=&TextBlock
   bIsVariable=true
   parent=PLACEMENTCHECKBOX
   widget=TextArrayAssets
   properties {
    text=Array Placement
    colorAndOpacity {colorUseRule=UseColor_Specified, specifiedColor=<rgba 0.85 0.85 0.85 1>}
    font {
     bForceMonospaced=false
     bMaterialIsStencil=false
     fontMaterial=null
     fontObject=&Roboto
     letterSpacing=0
     monospacedWidth=1
     size=9
     skewAmount=0
     typefaceFontName=Regular
     outlineSettings {
      bApplyOutlineToDropShadows=false
      bMiteredCorners=false
      bSeparateFillAlpha=false
      outlineMaterial=null
      outlineSize=0
      outlineColor=<rgba 0 0 0 1>
     }
    }
   }
   slot {class=HorizontalBoxSlot, properties {horizontalAlignment=HAlign_Left, verticalAlignment=VAlign_Center}}
  }
  Spacer_3 {widgetClassPath=&Spacer, parent=ARRAYHWRAPPER, widget=Spacer_3, properties {size=<16 1>}, slot {class=HorizontalBoxSlot}}
  ARRAYBOXTEXTINPUT {
   widgetClassPath=&HorizontalBox
   parent=ARRAYHWRAPPER
   widget=ARRAYBOXTEXTINPUT
   slot {class=HorizontalBoxSlot, properties {horizontalAlignment=HAlign_Left, size {sizeRule=Fill, value=1}}}
  }
  HBOXSPACING {
   widgetClassPath=&HorizontalBox
   parent=ARRAYBOXTEXTINPUT
   widget=HBOXSPACING
   slot {class=HorizontalBoxSlot, properties {horizontalAlignment=HAlign_Right}}
  }
  TextSpacing {
   widgetClassPath=&TextBlock
   bIsVariable=true
   parent=HBOXSPACING
   widget=TextSpacing
   properties {
    text=Spacing
    colorAndOpacity {colorUseRule=UseColor_Specified, specifiedColor=<rgba 0.85 0.85 0.85 1>}
    font {
     bForceMonospaced=false
     bMaterialIsStencil=false
     fontMaterial=null
     fontObject=&Roboto
     letterSpacing=0
     monospacedWidth=1
     size=9
     skewAmount=0
     typefaceFontName=Regular
     outlineSettings {
      bApplyOutlineToDropShadows=false
      bMiteredCorners=false
      bSeparateFillAlpha=false
      outlineMaterial=null
      outlineSize=0
      outlineColor=<rgba 0 0 0 1>
     }
    }
   }
   slot {class=HorizontalBoxSlot, properties {horizontalAlignment=HAlign_Left, verticalAlignment=VAlign_Center}}
  }
  InputSpacing {
   widgetClassPath=&EditorUtilityEditableTextBox
   bIsVariable=true
   parent=HBOXSPACING
   widget=InputSpacing
   properties {
    hintText=Enter unit amount to space placed actors
    keyboardType=Number
    revertTextOnEscape=true
    selectAllTextWhenFocused=true
    text=256
    toolTipText="Input a value in Unreal Units as spacing. 512 is Default Fortnite Grid. "
   }
   slot {class=HorizontalBoxSlot, properties {padding=<ltrb 8 0 8 0>}}
  }
  HBOXARRAYSIZE {
   widgetClassPath=&HorizontalBox
   parent=ARRAYBOXTEXTINPUT
   widget=HBOXARRAYSIZE
   slot {class=HorizontalBoxSlot, properties {horizontalAlignment=HAlign_Right, padding=<ltrb 16 0 8 0>}}
  }
  TextPerRow {
   widgetClassPath=&TextBlock
   bIsVariable=true
   parent=HBOXARRAYSIZE
   widget=TextPerRow
   properties {
    text=Per Row
    colorAndOpacity {colorUseRule=UseColor_Specified, specifiedColor=<rgba 0.85 0.85 0.85 1>}
    font {
     bForceMonospaced=false
     bMaterialIsStencil=false
     fontMaterial=null
     fontObject=&Roboto
     letterSpacing=0
     monospacedWidth=1
     size=9
     skewAmount=0
     typefaceFontName=Regular
     outlineSettings {
      bApplyOutlineToDropShadows=false
      bMiteredCorners=false
      bSeparateFillAlpha=false
      outlineMaterial=null
      outlineSize=0
      outlineColor=<rgba 0 0 0 1>
     }
    }
   }
   slot {class=HorizontalBoxSlot, properties {horizontalAlignment=HAlign_Left, verticalAlignment=VAlign_Center}}
  }
  InuptRows {
   widgetClassPath=&EditorUtilityEditableTextBox
   bIsVariable=true
   parent=HBOXARRAYSIZE
   widget=InuptRows
   properties {
    hintText=Enter unit amount to space placed actors
    keyboardType=Number
    revertTextOnEscape=true
    selectAllTextWhenFocused=true
    text=8
    toolTipText="Number of meshes to be present per row. "
   }
   slot {class=HorizontalBoxSlot, properties {padding=<ltrb 8 0 8 0>}}
  }
  Image {
   widgetClassPath=&Image
   bIsVariable=true
   parent=INPUTVWRAPPER
   widget=Image
   properties {brush {
     bIsDynamicallyLoaded=false
     drawAs=Image
     imageType=NoImage
     mirroring=NoMirror
     resourceName=None
     resourceObject=&T_White_srgb
     tiling=NoTile
     imageSize=<1 1>
     margin=<ltrb 0 0 0 0>
     outlineSettings {
      bUseBrushTransparency=false
      roundingType=HalfHeightRadius
      width=0
      color {colorUseRule=UseColor_Specified, specifiedColor=<rgba 0 0 0 0>}
      cornerRadii=<0 0 0 0>
     }
     tintColor {colorUseRule=UseColor_Specified, specifiedColor=<rgba 0 0 0 1>}
     uVRegion {bIsValid=false, max=<0 0>, min=<0 0>}
    }}
   slot {class=VerticalBoxSlot, properties {verticalAlignment=VAlign_Center}}
  }
 ]
}
